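# Creative My Island widget that displays one device binding as a horizontal row of source, other device and destination text.
# On Construct, Populate copies the Source, OtherDevice and Destination strings into the three CommonTextBlocks.
# The SourceIsEvent flag colors the two ends oppositely: an event source gets EventColor (orange) and its destination FunctionColor (blue), and the reverse otherwise.
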WidgetBlueprint /Game/Athena/UI/InGame/Creative/MyIsland/Cr_DvBrowserBinding
graphs [
 EventGraph {nodes [
   Event Construct()
   > Populate()
  ]}
 Populate {nodes [
   Populate() {is=title K2Node_FunctionEntry}
   > "SetText (Text)"(self=Get Text_Device() -> (Text_Device:Common Text Object Reference), InText=Get OtherDevice() -> (OtherDevice:Text))
   n1 = Get Text_Source() -> (Text_Source:Common Text Object Reference)
   > SetColorAndOpacity(
    self=@n1
    InColorAndOpacity=Select(
     Option 0:Slate Color Structure=Get FunctionColor() -> (FunctionColor:Slate Color Structure)
     Option 1:Slate Color Structure=Get EventColor() -> (EventColor:Slate Color Structure)
     Index=Get SourceIsEvent() -> (SourceIsEvent:bool)
    ) -> (ReturnValue:Slate Color Structure)
   )
   > "SetText (Text)"(self=@n1, InText=Get Source() -> (Source:Text))
   n2 = Get Text_Destination() -> (Text_Destination:Common Text Object Reference)
   > SetColorAndOpacity(self=@n2, InColorAndOpacity=Select(Option 0=Get EventColor(), Option 1=Get FunctionColor(), Index=Get SourceIsEvent()))
   > "SetText (Text)"(self=@n2, InText=Get Destination() -> (Destination:Text))
  ]}
]
variables [
 Destination {type=string, default=""}
 EventColor {type=SlateColor, default {colorUseRule=UseColor_Specified, specifiedColor=<rgba 0.7454 0.147 0.0006 1>}}
 FunctionColor {type=SlateColor, default {colorUseRule=UseColor_Specified, specifiedColor=<rgba 0 0.2086 0.807 1>}}
 OtherDevice {type=string, default=""}
 Source {type=string, default=""}
 SourceIsEvent {type=boolean}
]
widgets {
 info {inheritedWidgetCount=0, parentClass=&CommonUserWidget, rootWidgetClass=&HorizontalBox, widgetCount=4}
 widgets [
  HorizontalBox_0 {widgetClassPath=&HorizontalBox, widget=HorizontalBox_0}
  Text_Source {
   widgetClassPath=&CommonTextBlock
   bIsVariable=true
   parent=HorizontalBox_0
   widget=Text_Source
   properties {
    justification=Center
    minDesiredWidth=42
    text=Source
    font {
     bForceMonospaced=false
     bMaterialIsStencil=false
     fontMaterial=null
     fontObject=&BurbankSmall
     letterSpacing=0
     monospacedWidth=1
     size=18
     skewAmount=0
     typefaceFontName=Black
     outlineSettings {
      bApplyOutlineToDropShadows=false
      bMiteredCorners=false
      bSeparateFillAlpha=false
      outlineMaterial=null
      outlineSize=0
      outlineColor=<rgba 0 0 0 1>
     }
    }
    shadowOffset=<0 0>
   }
   slot {class=HorizontalBoxSlot, properties {verticalAlignment=VAlign_Center, padding=<ltrb 2 0 2 0>}}
  }
  Text_Device {
   widgetClassPath=&CommonTextBlock
   bIsVariable=true
   parent=HorizontalBox_0
   widget=Text_Device
   properties {
    text="Device]"
    colorAndOpacity {colorUseRule=UseColor_Specified, specifiedColor=<rgba 0.3185 0.7231 1 1>}
    font {
     bForceMonospaced=false
     bMaterialIsStencil=false
     fontMaterial=null
     fontObject=&BurbankSmall
     letterSpacing=0
     monospacedWidth=1
     size=18
     skewAmount=0
     typefaceFontName=Black
     outlineSettings {
      bApplyOutlineToDropShadows=false
      bMiteredCorners=false
      bSeparateFillAlpha=false
      outlineMaterial=null
      outlineSize=0
      outlineColor=<rgba 0 0 0 1>
     }
    }
    shadowOffset=<0 0>
   }
   slot {class=HorizontalBoxSlot, properties {horizontalAlignment=HAlign_Left, verticalAlignment=VAlign_Center, padding=<ltrb 12 0 12 0>}}
  }
  Text_Destination {
   widgetClassPath=&CommonTextBlock
   bIsVariable=true
   parent=HorizontalBox_0
   widget=Text_Destination
   properties {
    text="Destination]"
    colorAndOpacity {colorUseRule=UseColor_Specified, specifiedColor=<rgba 0.3185 0.7231 1 1>}
    font {
     bForceMonospaced=false
     bMaterialIsStencil=false
     fontMaterial=null
     fontObject=&BurbankSmall
     letterSpacing=0
     monospacedWidth=1
     size=18
     skewAmount=0
     typefaceFontName=Black
     outlineSettings {
      bApplyOutlineToDropShadows=false
      bMiteredCorners=false
      bSeparateFillAlpha=false
      outlineMaterial=null
      outlineSize=0
      outlineColor=<rgba 0 0 0 1>
     }
    }
    shadowOffset=<0 0>
   }
   slot {class=HorizontalBoxSlot, properties {verticalAlignment=VAlign_Center, size {sizeRule=Fill, value=1}}}
  }
 ]
}
