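# Button tile for a Creative My Island device browser, showing one device's name and its event and function bindings, with teleport to that device.
# Populate sets the name and fills a vertical list with binding widgets, flagged as events or functions.
# Selection and hover toggle a border and TELEPORT overlay.
# Clicking enables ghost and fly modes, calling server versions without authority, sends a TeleportToActor gameplay event carrying ActorRef, then broadcasts CloseUI.
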WidgetBlueprint /Game/Athena/UI/InGame/Creative/MyIsland/Cr_DvBrowserDeviceTile
graphs [
 BeginFlying {nodes [
   BeginFlying() {is=title K2Node_FunctionEntry}
   > n1 = Sequence()
   > n2 = Cast To FortPlayerPawnAthena(Object=GetOwningPlayerPawn!(self:User Widget Object Reference) -> (ReturnValue:Pawn Object Reference))
   > CreativeToggleGhost!(execute, self:Fort Player Pawn Athena Object Reference=@n2, bEnableOnly:bool=true) -> (then)
   > n3 = Branch(Condition=HasAuthority!(self:Actor Object Reference=@n2) -> (ReturnValue:bool))
   ServerCreativeToggleGhost!(execute=@n3.else, self:Fort Player Pawn Athena Object Reference=@n2, bEnableOnly:bool=true) -> (then)
   n4 = Cast To FortPlayerControllerGameplay!(execute=@n1.then_1, Object=GetOwningPlayer()) -> (then, CastFailed, AsFort Player Controller Gameplay:Fort Player Controller Gameplay Object Reference) {is=title K2Node_DynamicCast}
   > CreativeToggleFly!(execute, self:Fort Player Controller Gameplay Object Reference=@n4, bEnableOnly:bool=true) -> (then)
   > n5 = Branch(Condition=HasAuthority(self=@n4))
   ServerCreativeToggleFly!(execute=@n5.else, self:Fort Player Controller Gameplay Object Reference=@n4, bEnableOnly:bool=true) -> (then)
  ]}
 CloseUI {nodes [CloseUI() {is=title K2Node_FunctionEntry}]}
 EventGraph {nodes [
   Event On Selected()
   n1 = Get Image_Border() -> (Image_Border:Image Object Reference)
   > SetVisibility(self=@n1, InVisibility=SelfHitTestInvisible)
   n2 = Get Overlay_Teleport() -> (Overlay_Teleport:Overlay Object Reference)
   > SetVisibility(self=@n2, InVisibility=SelfHitTestInvisible)
   Event On Deselected()
   > SetVisibility(self=@n1, InVisibility=Collapsed)
   > SetVisibility(self=@n2, InVisibility=Hidden)
   Event On Hovered()
   > n3 = Branch(Condition=GetSelected!(self:Common Button Base Object Reference) -> (ReturnValue:bool))
   n4 = Get Image_Border()
   n5 = Get Overlay_Teleport()
   SetVisibility(execute=@n3.else, self=[@n4, @n5], InVisibility=SelfHitTestInvisible)
   Event On Unhovered()
   > n6 = Branch(Condition=GetSelected())
   SetVisibility(execute=@n6.else, self=[@n4, @n5], InVisibility=Collapsed)
   Event On Clicked()
   n7 = Get ActorRef() -> (ActorRef:Actor Object Reference)
   > PrintString!(
    execute
    InString="To String (Vector)"!(InVec:Vector=Get Actor Location(self=@n7)) -> (ReturnValue:String)
    bPrintToScreen
    bPrintToLog
    TextColor
    Duration
    Key
   ) -> (then)
   > BeginFlying() {is=title K2Node_CallFunction}
   n8 = GetOwningPlayerPawn(self=Self-Reference())
   > SendGameplayEventToActor!(
    execute
    Actor:Actor Object Reference=@n8
    EventTag:Gameplay Tag Structure=(TagName="Creative.Event.TeleportToActor")
    Payload:Gameplay Event Data Structure=Make Gameplay Event Data!(
     EventTag:Gameplay Tag Structure=(TagName="Creative.Event.TeleportToActor")
     Instigator:Actor Object Reference
     Target:Actor Object Reference=@n8
     OptionalObject:Object Reference=@n7
     OptionalObject2:Object Reference
     ContextHandle:Gameplay Effect Context Handle Structure=()
     InstigatorTags:Gameplay Tag Container Structure=(GameplayTags=)
     TargetTags:Gameplay Tag Container Structure=(GameplayTags=)
     EventMagnitude:float=0
     TargetData:Gameplay Ability Target Data Handle Structure=()
    ) -> (GameplayEventData:Gameplay Event Data Structure) {is=title K2Node_MakeStruct}
   ) -> (then)
   > Call Close UI!(execute:Exec, self:Cr Dv Browser Device Tile Object Reference) -> (then:Exec) {is=title K2Node_CallDelegate}
   Event OnMouseLeave!() -> (OutputDelegate, then, MouseEvent:"Pointer Event Structure (by ref)") {is=title K2Node_Event}
   Event OnMouseEnter()
  ]}
 Focus {nodes [Focus() {is=title K2Node_FunctionEntry}]}
 Populate {nodes [
   n1 = Populate!() -> (then, inName:Text, inEvents:Array of Texts, inFunctions:Array of Texts) {is=title K2Node_FunctionEntry}
   > "SetText (Text)"(self=Get Text_DeviceName() -> (Text_DeviceName:Common Text Object Reference), InText=@n1)
   > ClearChildren(self=Get VBox_Messages() -> (VBox_Messages:Vertical Box Object Reference))
   > n2 = For Each Loop(Array:Array of Texts=Get inEvents() -> (inEvents:Array of Texts)) -> (Array Element:Text)
   n3 = For Each Loop(Exec=@n2.Completed, Array=Get inFunctions() -> (inFunctions:Array of Texts))
   > n4 = Create Widget!(
    execute:Exec
    Class:User Widget Class Reference=&/Cr_DvBrowserBinding.Cr_DvBrowserBinding_C
    OwningPlayer:Player Controller Object Reference=GetOwningPlayer()
    Destination:Text
    SourceIsEvent:bool=False
    Source:Text=@n3
   ) -> (then:Exec, ReturnValue:Cr Dv Browser Binding Object Reference)
   > AddChildToVerticalBox(self=Get VBox_Messages(), Content=@n4)
   n5 = Create Widget(
    execute=@n2.LoopBody
    Class=&Cr_DvBrowserBinding_C
    OwningPlayer=GetOwningPlayer()
    SourceIsEvent=true
    Source=@n2
   )
   > AddChildToVerticalBox(self=Get VBox_Messages(), Content=@n5)
  ]}
 UpdateName {nodes [
   n1 = UpdateName!() -> (then, inName:Text) {is=title K2Node_FunctionEntry}
   > "SetText (Text)"(self=Get Text_DeviceName(), InText=@n1)
  ]}
]
variables [
 ActorRef {type=Actor, default=null}
 CloseUI {}
]
widgets {
 info {inheritedWidgetCount=1, parentClass=&CommonButtonLegacy, rootWidgetClass=&Overlay, widgetCount=18}
 widgets [
  Overlay_9 {widgetClassPath=&Overlay, widget=Overlay_9}
  Image {
   widgetClassPath=&Image
   parent=Overlay_9
   widget=Image
   properties {colorAndOpacity=<rgba 0 0.0168 0.0887 0.6>}
   slot {class=OverlaySlot, properties {horizontalAlignment=HAlign_Fill, verticalAlignment=VAlign_Fill}}
  }
  Image_Border {
   widgetClassPath=&Image
   bIsVariable=true
   parent=Overlay_9
   widget=Image_Border
   properties {
    visibility=Hidden
    brush {
     bIsDynamicallyLoaded=false
     drawAs=Box
     imageType=NoImage
     mirroring=NoMirror
     resourceName=None
     resourceObject=/FortUILibrary/Foundation/Buttons/Materials/Secondary/M-Button-TransparentList-Pressed
     tiling=NoTile
     imageSize=<32 32>
     margin=<ltrb 0.3 0.3 0.7 0.7>
     outlineSettings {
      bUseBrushTransparency=false
      roundingType=HalfHeightRadius
      width=0
      color {colorUseRule=UseColor_Specified, specifiedColor=<rgba 0 0 0 0>}
      cornerRadii=<0 0 0 0>
     }
     tintColor {colorUseRule=UseColor_Specified, specifiedColor=<rgba 0 0.5457 1 1>}
     uVRegion {bIsValid=false, max=<0 0>, min=<0 0>}
    }
   }
   slot {class=OverlaySlot, properties {horizontalAlignment=HAlign_Fill, verticalAlignment=VAlign_Fill}}
  }
  VerticalBox_0 {
   widgetClassPath=&VerticalBox
   parent=Overlay_9
   widget=VerticalBox_0
   slot {class=OverlaySlot, properties {horizontalAlignment=HAlign_Fill, verticalAlignment=VAlign_Fill, padding=<ltrb 4 4 4 4>}}
  }
  Overlay_0 {widgetClassPath=&Overlay, parent=VerticalBox_0, widget=Overlay_0, slot {class=VerticalBoxSlot}}
  HorizontalBox_0 {
   widgetClassPath=&HorizontalBox
   parent=Overlay_0
   widget=HorizontalBox_0
   slot {class=OverlaySlot, properties {horizontalAlignment=HAlign_Fill, padding=<ltrb 6 6 6 6>}}
  }
  Image_Device {
   widgetClassPath=&Image
   bIsVariable=true
   parent=HorizontalBox_0
   widget=Image_Device
   properties {visibility=Collapsed}
   slot {class=HorizontalBoxSlot, properties {horizontalAlignment=HAlign_Center, verticalAlignment=VAlign_Center}}
  }
  Text_DeviceName {
   widgetClassPath=&CommonTextBlock
   bIsVariable=true
   parent=HorizontalBox_0
   widget=Text_DeviceName
   properties {
    text=Name
    colorAndOpacity {colorUseRule=UseColor_Specified, specifiedColor=<rgba 0.6724 0.7835 0.9301 1>}
    font {
     bForceMonospaced=false
     bMaterialIsStencil=false
     fontMaterial=null
     fontObject=&BurbankSmall
     letterSpacing=0
     monospacedWidth=1
     size=16
     skewAmount=0
     typefaceFontName=Bold
     outlineSettings {
      bApplyOutlineToDropShadows=false
      bMiteredCorners=false
      bSeparateFillAlpha=false
      outlineMaterial=null
      outlineSize=0
      outlineColor=<rgba 0 0 0 1>
     }
    }
    shadowOffset=<0 0>
   }
   slot {class=HorizontalBoxSlot, properties {verticalAlignment=VAlign_Center, padding=<ltrb 4 0 0 0>, size {sizeRule=Fill, value=1}}}
  }
  Overlay_1 {
   widgetClassPath=&Overlay
   parent=VerticalBox_0
   widget=Overlay_1
   slot {class=VerticalBoxSlot, properties {size {sizeRule=Fill, value=1}}}
  }
  VBox_Messages {
   widgetClassPath=&VerticalBox
   bIsVariable=true
   parent=Overlay_1
   widget=VBox_Messages
   slot {class=OverlaySlot, properties {horizontalAlignment=HAlign_Fill, verticalAlignment=VAlign_Fill, padding=<ltrb 24 12 12 12>}}
  }
  HorizontalBox_1 {
   widgetClassPath=&HorizontalBox
   parent=Overlay_9
   widget=HorizontalBox_1
   slot {class=OverlaySlot, properties {horizontalAlignment=HAlign_Right, padding=<ltrb 12 12 18 12>}}
  }
  Overlay_Teleport {
   widgetClassPath=&Overlay
   bIsVariable=true
   parent=HorizontalBox_1
   widget=Overlay_Teleport
   properties {visibility=Hidden}
   slot {class=HorizontalBoxSlot}
  }
  Border_0 {
   widgetClassPath=&Border
   parent=Overlay_Teleport
   widget=Border_0
   properties {
    background {
     bIsDynamicallyLoaded=false
     drawAs=Image
     imageType=NoImage
     mirroring=NoMirror
     resourceName=None
     resourceObject=/Game/Athena/UI/InGame/Creative/Art/M_SimpleSkewBox_CreativeIslandListSelection
     tiling=NoTile
     imageSize=<32 32>
     margin=<ltrb 0 0 0 0>
     outlineSettings {
      bUseBrushTransparency=false
      roundingType=HalfHeightRadius
      width=0
      color {colorUseRule=UseColor_Specified, specifiedColor=<rgba 0 0 0 0>}
      cornerRadii=<0 0 0 0>
     }
     tintColor {colorUseRule=UseColor_Specified, specifiedColor=<rgba 1 1 1 1>}
     uVRegion {bIsValid=false, max=<0 0>, min=<0 0>}
    }
    padding=<ltrb 4 4 4 4>
    renderTransform {angle=0, scale=<1.3 1.3>, shear=<0 0>, translation=<0 0>}
   }
   slot {class=OverlaySlot, properties {horizontalAlignment=HAlign_Fill, verticalAlignment=VAlign_Fill}}
  }
  Image_140 {
   widgetClassPath=&Image
   parent=Border_0
   widget=Image_140
   properties {brush {
     bIsDynamicallyLoaded=false
     drawAs=Image
     imageType=NoImage
     mirroring=NoMirror
     resourceName=None
     resourceObject=&M_SimpleSkewBox_CreativeIslandListSelection
     tiling=NoTile
     imageSize=<32 32>
     margin=<ltrb 0 0 0 0>
     outlineSettings {
      bUseBrushTransparency=false
      roundingType=HalfHeightRadius
      width=0
      color {colorUseRule=UseColor_Specified, specifiedColor=<rgba 0 0 0 0>}
      cornerRadii=<0 0 0 0>
     }
     tintColor {colorUseRule=UseColor_Specified, specifiedColor=<rgba 0 0.5457 1 1>}
     uVRegion {bIsValid=false, max=<0 0>, min=<0 0>}
    }}
   slot {class=BorderSlot, properties {padding=<ltrb 4 4 4 4>}}
  }
  CommonTextBlock_116 {
   widgetClassPath=&CommonTextBlock
   parent=Overlay_Teleport
   widget=CommonTextBlock_116
   properties {
    style=/FortUILibrary/Styles/Typography/Foundation/TextStyle_BurbankSmall_S_White.TextStyle_BurbankSmall_S_White_C
    text=TELEPORT
    font {
     bForceMonospaced=false
     bMaterialIsStencil=false
     fontMaterial=null
     fontObject=&BurbankSmall
     letterSpacing=0
     monospacedWidth=1
     size=18
     skewAmount=0
     typefaceFontName=Black
     outlineSettings {
      bApplyOutlineToDropShadows=false
      bMiteredCorners=false
      bSeparateFillAlpha=false
      outlineMaterial=null
      outlineSize=0
      outlineColor=<rgba 0 0 0 1>
     }
    }
    shadowOffset=<0 0>
   }
   slot {class=OverlaySlot, properties {verticalAlignment=VAlign_Center, padding=<ltrb 12 12 12 12>}}
  }
  CommonVisibilityWidget_InputAction {
   widgetClassPath=&CommonVisibilityWidgetLegacy
   parent=HorizontalBox_1
   widget=CommonVisibilityWidget_InputAction
   properties {bShowForAndroid=false, bShowForIOS=false, bShowForMac=false, bShowForMouseAndKeyboard=false, bShowForTouch=false, padding=<ltrb 0 0 0 0>}
   slot {class=HorizontalBoxSlot, properties {horizontalAlignment=HAlign_Right, verticalAlignment=VAlign_Center, padding=<ltrb 12 0 0 0>}}
  }
  Overlay_InputAction {
   widgetClassPath=&Overlay
   bIsVariable=true
   parent=CommonVisibilityWidget_InputAction
   widget=Overlay_InputAction
   slot {class=BorderSlot, properties {padding=<ltrb 0 0 0 0>}}
  }
  InputActionWidget {
   widgetClassPath=&CommonActionWidget
   bInherited=true
   bIsVariable=true
   parent=Overlay_InputAction
   widget=InputActionWidget
   properties {
    progressMaterialParam=Percentage
    inputActions [{dataTable=&InputActionDataTable, rowName=GenericForward}]
    progressMaterialBrush {
     bIsDynamicallyLoaded=false
     drawAs=Image
     imageType=NoImage
     mirroring=NoMirror
     resourceName=None
     resourceObject=/Game/UI/Materials/HollowCircularProgressBar
     tiling=NoTile
     imageSize=<36 36>
     margin=<ltrb 0 0 0 0>
     outlineSettings {
      bUseBrushTransparency=false
      roundingType=HalfHeightRadius
      width=0
      color {colorUseRule=UseColor_Specified, specifiedColor=<rgba 0 0 0 0>}
      cornerRadii=<0 0 0 0>
     }
     tintColor {colorUseRule=UseColor_Specified, specifiedColor=<rgba 1 1 0 1>}
     uVRegion {bIsValid=false, max=<0 0>, min=<0 0>}
    }
   }
   slot {class=OverlaySlot}
  }
 ]
}
