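# List entry button for the Creative My Island device browser, showing a device's name and its event and function counts.
# On list item set it copies those values, the tracking GUID and resolved device actor from Obj_DvBrowserDeviceListItem.
# Hover or selection scrolls the name; selection also recolors it.
# Helper functions fill fields directly, send a TeleportToActor gameplay event, and enable ghost and fly modes, adding server calls without authority.
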WidgetBlueprint /Game/Athena/UI/InGame/Creative/MyIsland/Cr_DvBrowserDvEntry
graphs [
 BeginFlying {nodes [
   BeginFlying() {is=title K2Node_FunctionEntry}
   > n1 = Sequence()
   > n2 = Cast To FortPlayerPawnAthena(Object=GetOwningPlayerPawn!(self:User Widget Object Reference) -> (ReturnValue:Pawn Object Reference))
   > CreativeToggleGhost!(execute, self:Fort Player Pawn Athena Object Reference=@n2, bEnableOnly:bool=true) -> (then)
   > n3 = Branch(Condition=HasAuthority!(self:Actor Object Reference=@n2) -> (ReturnValue:bool))
   ServerCreativeToggleGhost!(execute=@n3.else, self:Fort Player Pawn Athena Object Reference=@n2, bEnableOnly:bool=true) -> (then)
   n4 = Cast To FortPlayerControllerGameplay!(execute=@n1.then_1, Object=GetOwningPlayer()) -> (then, CastFailed, AsFort Player Controller Gameplay:Fort Player Controller Gameplay Object Reference) {is=title K2Node_DynamicCast}
   > CreativeToggleFly!(execute, self:Fort Player Controller Gameplay Object Reference=@n4, bEnableOnly:bool=true) -> (then)
   > n5 = Branch(Condition=HasAuthority(self=@n4))
   ServerCreativeToggleFly!(execute=@n5.else, self:Fort Player Controller Gameplay Object Reference=@n4, bEnableOnly:bool=true) -> (then)
  ]}
 EventGraph {nodes [
   n1 = Event OnListItemObjectSet()
   > Is Valid(InputObject=@n1.ListItemObject)
   > n2 = Cast To Obj_DvBrowserDeviceListItem!(execute, Object=@n1.ListItemObject) -> (then, CastFailed, AsObj Dv Browser Device List Item:Obj Dv Browser Device List Item Object Reference) {is=title K2Node_DynamicCast}
   > "SetText (Text)"(
    self=Get Text_Device() -> (Text_Device:Common Text Object Reference)
    InText="To Text (String)"(InString=Get DisplayName!(self:Obj Dv Browser Device List Item Object Reference=@n2) -> (DisplayName:String))
   )
   > "SetText (Text)"(
    self=Get Text_NumEvents() -> (Text_NumEvents:Common Text Object Reference)
    InText="To Text (Integer)"(Value=Get NumEvents!(self:Obj Dv Browser Device List Item Object Reference=@n2) -> (NumEvents:int=0))
   )
   > "SetText (Text)"(
    self=Get Text_NumFunctions() -> (Text_NumFunctions:Common Text Object Reference)
    InText="To Text (Integer)"(Value=Get NumFunctions!(self:Obj Dv Browser Device List Item Object Reference=@n2) -> (NumFunctions:int=0))
   )
   > Set DeviceTrackingGUID(DeviceTrackingGUID:int=Get DeviceTrackingGUID!(self:Obj Dv Browser Device List Item Object Reference=@n2) -> (DeviceTrackingGUID:int=0)) -> (Output_Get:int=0)
   > Set DeviceActor(DeviceActor:Actor Object Reference=Resolve Soft Reference!(Input:Actor Soft Object Reference=Get DeviceActor!(self:Obj Dv Browser Device List Item Object Reference=@n2) -> (DeviceActor:Actor Soft Object Reference)) -> (Output:Actor Object Reference)) -> (Output_Get:Actor Object Reference)
   n3 = Event On Hovered()
   n4 = Get Text_Device()
   SetScrollingEnabled!(execute=[@n7.then, @n3.then], self:Common Text Object Reference=@n4, bInIsScrollingEnabled:bool=true) -> (then)
   n5 = Event On Unhovered()
   SetScrollingEnabled(execute=[@n8.then, @n5.then], self=@n4, bInIsScrollingEnabled=false)
   Event On Selected()
   n6 = Get Text_Device()
   > n7 = Set ColorAndOpacity!(
    execute
    ColorAndOpacity:Slate Color Structure=Get DeviceLabelSelected_Color() -> (DeviceLabelSelected_Color:Slate Color Structure)
    self:Text Object Reference=@n6
   ) -> (then, Output_Get:Slate Color Structure)
   Event On Deselected()
   > n8 = Set ColorAndOpacity(ColorAndOpacity=Get DeviceLabelNormal_Color() -> (DeviceLabelNormal_Color:Slate Color Structure), self=@n6)
  ]}
 Populate {nodes [
   n1 = Populate!() -> (then, InDeviceTrackingGUID:int, inName:Text, inNumEvents:Text, inNumFunctions:Text) {is=title K2Node_FunctionEntry}
   > Set DeviceTrackingGUID(DeviceTrackingGUID=@n1) -> (Output_Get=0)
   > "SetText (Text)"(self=Get Text_Device(), InText=@n1.inName)
   > "SetText (Text)"(self=Get Text_NumEvents(), InText=@n1.inNumEvents)
   > "SetText (Text)"(self=Get Text_NumFunctions(), InText=@n1.inNumFunctions)
  ]}
 Teleport {nodes [
   n1 = Teleport!() -> (then, InActor:Object Reference) {is=title K2Node_FunctionEntry}
   n2 = GetOwningPlayerPawn(self=Self-Reference())
   > SendGameplayEventToActor!(
    execute
    Actor:Actor Object Reference=@n2
    EventTag:Gameplay Tag Structure=(TagName="Creative.Event.TeleportToActor")
    Payload:Gameplay Event Data Structure=Make Gameplay Event Data!(
     EventTag:Gameplay Tag Structure=(TagName="Creative.Event.TeleportToActor")
     Instigator:Actor Object Reference
     Target:Actor Object Reference=@n2
     OptionalObject:Object Reference=@n1
     OptionalObject2:Object Reference
     ContextHandle:Gameplay Effect Context Handle Structure=()
     InstigatorTags:Gameplay Tag Container Structure=(GameplayTags=)
     TargetTags:Gameplay Tag Container Structure=(GameplayTags=)
     EventMagnitude:float=0
     TargetData:Gameplay Ability Target Data Handle Structure=()
    ) -> (GameplayEventData:Gameplay Event Data Structure) {is=title K2Node_MakeStruct}
   ) -> (then)
  ]}
]
variables [
 DeviceActor {type=Actor, default=null}
 DeviceLabelNormal_Color {type=SlateColor, default {colorUseRule=UseColor_Specified, specifiedColor=<rgba 1 1 1 1>}}
 DeviceLabelSelected_Color {type=SlateColor, default {colorUseRule=UseColor_Specified, specifiedColor=<rgba 0 0 0 1>}}
 DeviceTrackingGUID {type=integer}
 Display_NumEvents {type=integer}
 Display_NumFuncts {type=integer}
 DisplayName {type=string, default=""}
]
widgets {
 info {inheritedWidgetCount=1, parentClass=&CommonButtonBase, rootWidgetClass=&Overlay, widgetCount=14}
 widgets [
  Overlay_9 {widgetClassPath=&Overlay, widget=Overlay_9}
  Spacer_LineHeight {
   widgetClassPath=&Spacer
   parent=Overlay_9
   widget=Spacer_LineHeight
   properties {size=<1 80>}
   slot {class=OverlaySlot, properties {horizontalAlignment=HAlign_Fill}}
  }
  Border_47 {
   widgetClassPath=&Border
   parent=Overlay_9
   widget=Border_47
   properties {visibility=Collapsed, brushColor=<rgba 0.0024 0.0513 0.3864 1>, padding=<ltrb 0 0 0 0>}
   slot {class=OverlaySlot, properties {horizontalAlignment=HAlign_Fill, verticalAlignment=VAlign_Fill}}
  }
  Hovered_Border {
   widgetClassPath=&Border
   bIsVariable=true
   parent=Overlay_9
   widget=Hovered_Border
   properties {
    visibility=Collapsed
    background {
     bIsDynamicallyLoaded=false
     drawAs=Image
     imageType=NoImage
     mirroring=NoMirror
     resourceName=None
     resourceObject=null
     tiling=NoTile
     imageSize=<32 32>
     margin=<ltrb 0 0 0 0>
     outlineSettings {
      bUseBrushTransparency=false
      roundingType=HalfHeightRadius
      width=0
      color {colorUseRule=UseColor_Specified, specifiedColor=<rgba 0 0 0 0>}
      cornerRadii=<0 0 0 0>
     }
     tintColor {colorUseRule=UseColor_Specified, specifiedColor=<rgba 0.0044 0.3712 0.8308 1>}
     uVRegion {bIsValid=false, max=<0 0>, min=<0 0>}
    }
    padding=<ltrb 0 0 0 0>
   }
   slot {class=OverlaySlot, properties {horizontalAlignment=HAlign_Fill, verticalAlignment=VAlign_Fill}}
  }
  Image_Selection_Focused {
   widgetClassPath=&Image
   bIsVariable=true
   parent=Overlay_9
   widget=Image_Selection_Focused
   properties {
    visibility=Collapsed
    brush {
     bIsDynamicallyLoaded=false
     drawAs=Image
     imageType=NoImage
     mirroring=NoMirror
     resourceName=None
     resourceObject=null
     tiling=NoTile
     imageSize=<0 0>
     margin=<ltrb 0 0 0 0>
     outlineSettings {
      bUseBrushTransparency=false
      roundingType=HalfHeightRadius
      width=0
      color {colorUseRule=UseColor_Specified, specifiedColor=<rgba 0 0 0 0>}
      cornerRadii=<0 0 0 0>
     }
     tintColor {colorUseRule=UseColor_Specified, specifiedColor=<rgba 1 1 1 1>}
     uVRegion {bIsValid=false, max=<0 0>, min=<0 0>}
    }
    renderTransform {angle=0, scale=<1.35 1>, shear=<0 0>, translation=<0 0>}
   }
   slot {class=OverlaySlot, properties {horizontalAlignment=HAlign_Fill, verticalAlignment=VAlign_Fill, padding=<ltrb -16 0 -16 0>}}
  }
  Overlay_1 {
   widgetClassPath=&Overlay
   parent=Overlay_9
   widget=Overlay_1
   slot {class=OverlaySlot, properties {horizontalAlignment=HAlign_Fill, verticalAlignment=VAlign_Fill, padding=<ltrb 0 6 24 6>}}
  }
  HorizontalBox_0 {
   widgetClassPath=&HorizontalBox
   parent=Overlay_1
   widget=HorizontalBox_0
   slot {class=OverlaySlot, properties {verticalAlignment=VAlign_Center, padding=<ltrb 6 0 0 0>}}
  }
  CommonVisibilityWidget_InputAction {
   widgetClassPath=&CommonVisibilityWidgetLegacy
   parent=HorizontalBox_0
   widget=CommonVisibilityWidget_InputAction
   properties {bShowForAndroid=false, bShowForIOS=false, bShowForMac=false, bShowForMouseAndKeyboard=false, bShowForTouch=false, padding=<ltrb 0 0 0 0>}
   slot {class=HorizontalBoxSlot, properties {horizontalAlignment=HAlign_Left, verticalAlignment=VAlign_Center, padding=<ltrb 0 0 12 0>}}
  }
  Overlay_InputAction {
   widgetClassPath=&Overlay
   bIsVariable=true
   parent=CommonVisibilityWidget_InputAction
   widget=Overlay_InputAction
   slot {class=BorderSlot, properties {padding=<ltrb 0 0 0 0>}}
  }
  InputActionWidget {
   widgetClassPath=&CommonActionWidget
   bInherited=true
   bIsVariable=true
   parent=Overlay_InputAction
   widget=InputActionWidget
   properties {
    progressMaterialParam=Percentage
    inputActions [{dataTable=&InputActionDataTable, rowName=GenericForward}]
    progressMaterialBrush {
     bIsDynamicallyLoaded=false
     drawAs=Image
     imageType=NoImage
     mirroring=NoMirror
     resourceName=None
     resourceObject=/Game/UI/Materials/HollowCircularProgressBar
     tiling=NoTile
     imageSize=<36 36>
     margin=<ltrb 0 0 0 0>
     outlineSettings {
      bUseBrushTransparency=false
      roundingType=HalfHeightRadius
      width=0
      color {colorUseRule=UseColor_Specified, specifiedColor=<rgba 0 0 0 0>}
      cornerRadii=<0 0 0 0>
     }
     tintColor {colorUseRule=UseColor_Specified, specifiedColor=<rgba 1 1 0 1>}
     uVRegion {bIsValid=false, max=<0 0>, min=<0 0>}
    }
   }
   slot {class=OverlaySlot}
  }
  SizeBox_0 {
   widgetClassPath=&SizeBox
   parent=HorizontalBox_0
   widget=SizeBox_0
   properties {bOverride_WidthOverride=true, widthOverride=340}
   slot {class=HorizontalBoxSlot}
  }
  Text_Device {
   widgetClassPath=&CommonTextBlock
   bIsVariable=true
   parent=SizeBox_0
   widget=Text_Device
   properties {
    bIsScrollingEnabled=false
    scrollStyle=&TextScrollStyle-Base_C
    text=Device Name
    font {
     bForceMonospaced=false
     bMaterialIsStencil=false
     fontMaterial=null
     fontObject=&BurbankSmall
     letterSpacing=0
     monospacedWidth=1
     size=18
     skewAmount=0
     typefaceFontName=Bold
     outlineSettings {
      bApplyOutlineToDropShadows=false
      bMiteredCorners=false
      bSeparateFillAlpha=false
      outlineMaterial=null
      outlineSize=0
      outlineColor=<rgba 0 0 0 1>
     }
    }
   }
   slot {class=SizeBoxSlot, properties {padding=<ltrb 24 0 0 0>}}
  }
  Text_NumEvents {
   widgetClassPath=&CommonTextBlock
   bIsVariable=true
   parent=Overlay_1
   widget=Text_NumEvents
   properties {
    justification=Center
    text=Events
    colorAndOpacity {colorUseRule=UseColor_Specified, specifiedColor=<rgba 0.8308 0.1651 0 1>}
    font {
     bForceMonospaced=false
     bMaterialIsStencil=false
     fontMaterial=null
     fontObject=&BurbankSmall
     letterSpacing=0
     monospacedWidth=1
     size=18
     skewAmount=0
     typefaceFontName=Bold
     outlineSettings {
      bApplyOutlineToDropShadows=false
      bMiteredCorners=false
      bSeparateFillAlpha=false
      outlineMaterial=null
      outlineSize=0
      outlineColor=<rgba 0 0 0 0.1>
     }
    }
   }
   slot {class=OverlaySlot, properties {horizontalAlignment=HAlign_Center, verticalAlignment=VAlign_Center, padding=<ltrb 200 0 0 0>}}
  }
  Text_NumFunctions {
   widgetClassPath=&CommonTextBlock
   bIsVariable=true
   parent=Overlay_1
   widget=Text_NumFunctions
   properties {
    justification=Center
    text=Rx
    colorAndOpacity {colorUseRule=UseColor_Specified, specifiedColor=<rgba 0.0044 0.3712 0.8308 1>}
    font {
     bForceMonospaced=false
     bMaterialIsStencil=false
     fontMaterial=null
     fontObject=&BurbankSmall
     letterSpacing=0
     monospacedWidth=1
     size=18
     skewAmount=0
     typefaceFontName=Bold
     outlineSettings {
      bApplyOutlineToDropShadows=false
      bMiteredCorners=false
      bSeparateFillAlpha=false
      outlineMaterial=null
      outlineSize=0
      outlineColor=<rgba 0 0 0 0.1>
     }
    }
   }
   slot {class=OverlaySlot, properties {horizontalAlignment=HAlign_Right, verticalAlignment=VAlign_Center, padding=<ltrb 0 0 12 0>}}
  }
 ]
}
