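# List entry in the Creative device browser that shows a device's event/function subscription, a section header or a name title.
# OnListItemObjectSet copies flags from the entry object, treating named items with GUID -1 as titles, then SetFromSubscription fills the texts.
# SetState colors by event/function, picks the switcher panel and makes headers and titles hit-test invisible.
# Hover and selection show backing highlights and scroll the row's text.
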WidgetBlueprint /Game/Athena/UI/InGame/Creative/MyIsland/Cr_DvBrowser_SubscriptionEntry
graphs [
 EventGraph {nodes [
   Event On Hovered()
   > SetVisibility(self=Get Image_Backing() -> (Image_Backing:Image Object Reference), InVisibility=SelfHitTestInvisible)
   n1 = Get Text_Source() -> (Text_Source:Common Text Object Reference)
   n2 = Get Text_Destination() -> (Text_Destination:Common Text Object Reference)
   n3 = Get Text_DestinationDevice() -> (Text_DestinationDevice:Common Text Object Reference)
   > SetScrollingEnabled!(execute, self:Common Text Object Reference=[@n1, @n2, @n3], bInIsScrollingEnabled:bool=true) -> (then)
   Event On Unhovered()
   > SetVisibility(self=Get Image_Backing(), InVisibility=Hidden)
   > SetScrollingEnabled(self=[@n1, @n2, @n3], bInIsScrollingEnabled=false)
   Event On Deselected()
   > SetVisibility(self=Get Image_Backing_Selected() -> (Image_Backing_Selected:Image Object Reference), InVisibility=Hidden)
   n4 = Get Text_Source()
   n5 = Get Text_Destination()
   n6 = Get Text_DestinationDevice()
   > SetScrollingEnabled(self=[@n4, @n5, @n6], bInIsScrollingEnabled=false)
   Event On Selected()
   > SetVisibility(self=Get Image_Backing_Selected(), InVisibility=SelfHitTestInvisible)
   > SetScrollingEnabled(self=[@n4, @n5, @n6], bInIsScrollingEnabled=true)
   n7 = Event OnListItemObjectSet()
   > n8 = Set Item(Item:Object Reference=@n7.ListItemObject) -> (Output_Get:Object Reference)
   > Is Valid(InputObject=@n8)
   > n9 = Cast To EventBrowserSubscriptionListEntryObj!(execute, Object=@n8) -> (then, CastFailed, AsEvent Browser Subscription List Entry Obj:Event Browser Subscription List Entry Obj Object Reference) {is=title K2Node_DynamicCast}
   > Set bIsHeader(bIsHeader:bool=Get bIsHeader!(self:Event Browser Subscription List Entry Obj Object Reference=@n9) -> (bIsHeader:bool)) -> (Output_Get:bool)
   > Set bIsEvent(bIsEvent:bool=Get bIsEvent!(self:Event Browser Subscription List Entry Obj Object Reference=@n9) -> (bIsEvent:bool)) -> (Output_Get:bool)
   > Set DeviceTrackingGUID(DeviceTrackingGUID:int=Get DeviceTrackingGUID!(self:Event Browser Subscription List Entry Obj Object Reference=@n9) -> (DeviceTrackingGUID:int=0)) -> (Output_Get:int=0)
   > n10 = Set DisplayName(DisplayName:String=Get DisplayName!(self:Event Browser Subscription List Entry Obj Object Reference=@n9) -> (DisplayName:String)) -> (Output_Get:String)
   n11 = IsEmpty(InString=@n10)
   n12 = Get DeviceTrackingGUID!() -> (DeviceTrackingGUID=0)
   > Set bIsMessage(bIsMessage:bool=AND Boolean(A=NOT Boolean(A=@n11), B="Equal (Integer)"(A=@n12, B=-1))) -> (Output_Get:bool)
   > SetFromSubscription!(
    execute
    self
    SourceSubscription:Creative Device Subscription Structure=Get Subscription!(self:Event Browser Subscription List Entry Obj Object Reference=@n9) -> (Subscription:Creative Device Subscription Structure)
   ) -> (then)
   PrintString!(
    execute=@n9.CastFailed
    InString=Append(A="Cast Failed: ", B=GetDisplayName(Object=@n8))
    bPrintToScreen
    bPrintToLog
    TextColor
    Duration
    Key
   ) -> (then)
  ]}
 SetFromSubscription {nodes [
   n1 = SetFromSubscription!() -> (then, SourceSubscription:Creative Device Subscription Structure) {is=title K2Node_FunctionEntry}
   n2 = Break Creative Device Subscription!(CreativeDeviceSubscription:"Creative Device Subscription Structure (by ref)"=@n1) -> (
    DestinationDeviceGUID:int
    bIsEvent:bool=~
    SourceSubscriptionDisplayName:Text
    DestinationDeviceDisplayName:Text
    DestinationSubscriptionDisplayName:Text
    ActorPtr:Actor Soft Object Reference
   ) {is=title K2Node_BreakStruct}
   > "SetText (Text)"(self=Get Text_Source(), InText=@n2.SourceSubscriptionDisplayName)
   > "SetText (Text)"(self=Get Text_Destination(), InText=@n2.DestinationDeviceDisplayName)
   > "SetText (Text)"(self=Get Text_DestinationDevice(), InText=@n2.DestinationSubscriptionDisplayName)
   > "SetText (Text)"(self=Get Name_Text() -> (Name_Text:Common Text Object Reference), InText="To Text (String)"(InString=Get DisplayName!() -> (DisplayName)))
   > SetState()
  ]}
 SetState {nodes [
   SetState() {is=title K2Node_FunctionEntry}
   > SetColorAndOpacity(
    self=Get Text_DestinationDevice()
    InColorAndOpacity=Select(
     Option 0:Slate Color Structure=Get FunctionColor() -> (FunctionColor:Slate Color Structure)
     Option 1:Slate Color Structure=Get EventColor() -> (EventColor:Slate Color Structure)
     Index=Get bIsEvent!() -> (bIsEvent)
    ) -> (ReturnValue:Slate Color Structure)
   )
   > SetColorAndOpacity(self=Get Text_Source(), InColorAndOpacity=Select(Option 0=Get EventColor(), Option 1=Get FunctionColor(), Index=Get bIsEvent()))
   n1 = Select(
    Option 0:int=Get FunctionHeaderPanel() -> (FunctionHeaderPanel:int=0)
    Option 1:int=Get EventHeaderPanel() -> (EventHeaderPanel:int=0)
    Index=Get bIsEvent()
   ) -> (ReturnValue:int)
   n2 = Get ItemPanel() -> (ItemPanel:int=0)
   n3 = Get bIsHeader!() -> (bIsHeader)
   > SetActiveWidgetIndex(
    self=Get Switcher() -> (Switcher:Common Animated Switcher Object Reference)
    Index=Select(
     Option 0=Select(Option 0=@n2, Option 1=@n1, Index=@n3)
     Option 1=Get MessagePanel() -> (MessagePanel:int=0)
     Index=Get bIsMessage() -> (bIsMessage:bool)
    )
   )
   n4 = Get bIsHeader()
   n5 = Get bIsMessage()
   > SetVisibility(InVisibility=Select(Option 0:ESlateVisibility Enum=Visible, Option 1:ESlateVisibility Enum=HitTestInvisible, Index=OR Boolean(A=@n4, B=@n5)) -> (ReturnValue:ESlateVisibility Enum))
  ]}
]
variables [
 bIsEvent {type=boolean}
 bIsHeader {type=boolean}
 bIsMessage {type=boolean}
 DestinationDevice_Color {type=SlateColor, default {colorUseRule=UseColor_Specified, specifiedColor=<rgba 0.3372 0.402 0.5906 1>}}
 DeviceLabelNormal_Color {type=SlateColor, default {colorUseRule=UseColor_Specified, specifiedColor=<rgba 1 1 1 1>}}
 DeviceLabelSelected_Color {type=SlateColor, default {colorUseRule=UseColor_Specified, specifiedColor=<rgba 0 0 0 1>}}
 DeviceTrackingGUID {type=integer}
 Display_NumEvents {type=integer}
 Display_NumFuncts {type=integer}
 DisplayName {type=string, default=""}
 Event_Color {type=SlateColor, default {colorUseRule=UseColor_Specified, specifiedColor=<rgba 1 0.2122 0.003 1>}}
 EventHeaderPanel {type=integer, default=1}
 Function_Color {type=SlateColor, default {colorUseRule=UseColor_Specified, specifiedColor=<rgba 0 0.2384 0.7991 1>}}
 FunctionHeaderPanel {type=integer, default=2}
 Item {type=Object, default=null}
 ItemPanel {type=integer}
 MessagePanel {type=integer, default=3}
 Subscription {
  type=CreativeDeviceSubscription
  default {
   bIsEvent=false
   destinationDeviceDisplayName=""
   destinationDeviceGUId=0
   destinationSubscriptionDisplayName=""
   sourceSubscriptionDisplayName=""
  }
 }
 TeleportCoords {type=Vector_NetQuantize10, default=<0 0 0>}
]
widgets {
 info {
  inheritedWidgetCount=1
  parentClass=/Script/FortniteUI.FortCreativeDeviceBrowserListEntry
  rootWidgetClass=&Overlay
  widgetCount=33
 }
 widgets [
  Overlay_0 {widgetClassPath=&Overlay, widget=Overlay_0}
  Switcher {
   widgetClassPath=/Script/CommonUI.CommonAnimatedSwitcher
   bIsVariable=true
   parent=Overlay_0
   widget=Switcher
   properties {transitionDuration=0}
   slot {class=OverlaySlot, properties {horizontalAlignment=HAlign_Fill, verticalAlignment=VAlign_Fill}}
  }
  RowItem_Overlay {widgetClassPath=&Overlay, parent=Switcher, widget=RowItem_Overlay, slot {class=WidgetSwitcherSlot}}
  Spacer_RowHeight {
   widgetClassPath=&Spacer
   parent=RowItem_Overlay
   widget=Spacer_RowHeight
   properties {size=<1 80>}
   slot {class=OverlaySlot, properties {horizontalAlignment=HAlign_Fill, verticalAlignment=VAlign_Fill}}
  }
  Image_3 {
   widgetClassPath=&Image
   parent=RowItem_Overlay
   widget=Image_3
   properties {
    visibility=SelfHitTestInvisible
    brush {
     bIsDynamicallyLoaded=false
     drawAs=Image
     imageType=NoImage
     mirroring=NoMirror
     resourceName=None
     resourceObject=null
     tiling=NoTile
     imageSize=<32 2>
     margin=<ltrb 0 0 0 0>
     outlineSettings {
      bUseBrushTransparency=false
      roundingType=HalfHeightRadius
      width=0
      color {colorUseRule=UseColor_Specified, specifiedColor=<rgba 0 0 0 0>}
      cornerRadii=<0 0 0 0>
     }
     tintColor {colorUseRule=UseColor_Specified, specifiedColor=<rgba 1 1 1 1>}
     uVRegion {bIsValid=false, max=<0 0>, min=<0 0>}
    }
    colorAndOpacity=<rgba 0.0024 0.016 0.117 1>
   }
   slot {class=OverlaySlot, properties {horizontalAlignment=HAlign_Fill, verticalAlignment=VAlign_Fill}}
  }
  HorizontalBox_0 {
   widgetClassPath=&HorizontalBox
   parent=RowItem_Overlay
   widget=HorizontalBox_0
   slot {class=OverlaySlot, properties {horizontalAlignment=HAlign_Fill, verticalAlignment=VAlign_Fill, padding=<ltrb 24 6 24 6>}}
  }
  CommonVisibilityWidget_InputAction {
   widgetClassPath=&CommonVisibilityWidgetLegacy
   parent=HorizontalBox_0
   widget=CommonVisibilityWidget_InputAction
   properties {bShowForAndroid=false, bShowForIOS=false, bShowForMac=false, bShowForMouseAndKeyboard=false, bShowForTouch=false, padding=<ltrb 0 0 0 0>}
   slot {class=HorizontalBoxSlot, properties {horizontalAlignment=HAlign_Center, verticalAlignment=VAlign_Center}}
  }
  Overlay_InputAction {
   widgetClassPath=&Overlay
   bIsVariable=true
   parent=CommonVisibilityWidget_InputAction
   widget=Overlay_InputAction
   slot {class=BorderSlot, properties {padding=<ltrb 0 0 0 0>}}
  }
  InputActionWidget {
   widgetClassPath=&CommonActionWidget
   bInherited=true
   bIsVariable=true
   parent=Overlay_InputAction
   widget=InputActionWidget
   properties {
    progressMaterialParam=Percentage
    inputActions [{dataTable=/FortUILibrary/Input/CreativeInputActionDataTable, rowName=Teleport}]
    progressMaterialBrush {
     bIsDynamicallyLoaded=false
     drawAs=Image
     imageType=NoImage
     mirroring=NoMirror
     resourceName=None
     resourceObject=/Game/UI/Materials/HollowCircularProgressBar
     tiling=NoTile
     imageSize=<36 36>
     margin=<ltrb 0 0 0 0>
     outlineSettings {
      bUseBrushTransparency=false
      roundingType=HalfHeightRadius
      width=0
      color {colorUseRule=UseColor_Specified, specifiedColor=<rgba 0 0 0 0>}
      cornerRadii=<0 0 0 0>
     }
     tintColor {colorUseRule=UseColor_Specified, specifiedColor=<rgba 1 1 0 1>}
     uVRegion {bIsValid=false, max=<0 0>, min=<0 0>}
    }
   }
   slot {class=OverlaySlot}
  }
  SizeBox_0 {
   widgetClassPath=&SizeBox
   parent=HorizontalBox_0
   widget=SizeBox_0
   slot {class=HorizontalBoxSlot, properties {verticalAlignment=VAlign_Center, padding=<ltrb 12 0 18 0>, size {sizeRule=Fill, value=1}}}
  }
  Text_DestinationDevice {
   widgetClassPath=&CommonTextBlock
   bIsVariable=true
   parent=SizeBox_0
   widget=Text_DestinationDevice
   properties {
    bIsScrollingEnabled=false
    scrollStyle=&TextScrollStyle-Base_C
    text=DestinationDevice
    colorAndOpacity {colorUseRule=UseColor_Specified, specifiedColor=<rgba 0.8308 0.1651 0 1>}
    font {
     bForceMonospaced=false
     bMaterialIsStencil=false
     fontMaterial=null
     fontObject=&BurbankSmall
     letterSpacing=0
     monospacedWidth=1
     size=18
     skewAmount=0
     typefaceFontName=Bold
     outlineSettings {
      bApplyOutlineToDropShadows=false
      bMiteredCorners=false
      bSeparateFillAlpha=false
      outlineMaterial=null
      outlineSize=0
      outlineColor=<rgba 0 0 0 0.1>
     }
    }
   }
   slot {class=SizeBoxSlot}
  }
  SizeBox_1 {
   widgetClassPath=&SizeBox
   parent=HorizontalBox_0
   widget=SizeBox_1
   slot {class=HorizontalBoxSlot, properties {verticalAlignment=VAlign_Center, padding=<ltrb 0 0 18 0>, size {sizeRule=Fill, value=1}}}
  }
  Text_Destination {
   widgetClassPath=&CommonTextBlock
   bIsVariable=true
   parent=SizeBox_1
   widget=Text_Destination
   properties {
    bIsScrollingEnabled=false
    scrollStyle=&TextScrollStyle-Base_C
    text=Destination
    font {
     bForceMonospaced=false
     bMaterialIsStencil=false
     fontMaterial=null
     fontObject=&BurbankSmall
     letterSpacing=0
     monospacedWidth=1
     size=18
     skewAmount=0
     typefaceFontName=Bold
     outlineSettings {
      bApplyOutlineToDropShadows=false
      bMiteredCorners=false
      bSeparateFillAlpha=false
      outlineMaterial=null
      outlineSize=0
      outlineColor=<rgba 0 0 0 0.1>
     }
    }
   }
   slot {class=SizeBoxSlot}
  }
  SizeBox_2 {
   widgetClassPath=&SizeBox
   parent=HorizontalBox_0
   widget=SizeBox_2
   slot {class=HorizontalBoxSlot, properties {verticalAlignment=VAlign_Center, padding=<ltrb 0 0 12 0>, size {sizeRule=Fill, value=1}}}
  }
  Text_Source {
   widgetClassPath=&CommonTextBlock
   bIsVariable=true
   parent=SizeBox_2
   widget=Text_Source
   properties {
    bIsScrollingEnabled=false
    scrollStyle=&TextScrollStyle-Base_C
    text=Source
    colorAndOpacity {colorUseRule=UseColor_Specified, specifiedColor=<rgba 0.0044 0.3712 0.8308 1>}
    font {
     bForceMonospaced=false
     bMaterialIsStencil=false
     fontMaterial=null
     fontObject=&BurbankSmall
     letterSpacing=0
     monospacedWidth=1
     size=18
     skewAmount=0
     typefaceFontName=Bold
     outlineSettings {
      bApplyOutlineToDropShadows=false
      bMiteredCorners=false
      bSeparateFillAlpha=false
      outlineMaterial=null
      outlineSize=0
      outlineColor=<rgba 0 0 0 1>
     }
    }
   }
   slot {class=SizeBoxSlot}
  }
  Header_Functions_Overlay {widgetClassPath=&Overlay, parent=Switcher, widget=Header_Functions_Overlay, slot {class=WidgetSwitcherSlot}}
  Image_4 {
   widgetClassPath=&Image
   parent=Header_Functions_Overlay
   widget=Image_4
   properties {
    visibility=Collapsed
    brush {
     bIsDynamicallyLoaded=false
     drawAs=Image
     imageType=NoImage
     mirroring=NoMirror
     resourceName=None
     resourceObject=null
     tiling=NoTile
     imageSize=<32 2>
     margin=<ltrb 0 0 0 0>
     outlineSettings {
      bUseBrushTransparency=false
      roundingType=HalfHeightRadius
      width=0
      color {colorUseRule=UseColor_Specified, specifiedColor=<rgba 0 0 0 0>}
      cornerRadii=<0 0 0 0>
     }
     tintColor {colorUseRule=UseColor_Specified, specifiedColor=<rgba 1 1 1 1>}
     uVRegion {bIsValid=false, max=<0 0>, min=<0 0>}
    }
    colorAndOpacity=<rgba 0.0024 0.016 0.117 1>
   }
   slot {class=OverlaySlot, properties {horizontalAlignment=HAlign_Fill, verticalAlignment=VAlign_Fill}}
  }
  HorizontalBox_85 {
   widgetClassPath=&HorizontalBox
   parent=Header_Functions_Overlay
   widget=HorizontalBox_85
   slot {class=OverlaySlot, properties {horizontalAlignment=HAlign_Fill, verticalAlignment=VAlign_Bottom, padding=<ltrb 12 32 12 12>}}
  }
  Image_194 {
   widgetClassPath=&Image
   parent=HorizontalBox_85
   widget=Image_194
   properties {brush {
     bIsDynamicallyLoaded=false
     drawAs=Image
     imageType=NoImage
     mirroring=NoMirror
     resourceName=None
     resourceObject=/Game/Athena/UI/InGame/Creative/PropertyEditor/Art/MI_UI_Binding_FunctionNormal
     tiling=NoTile
     imageSize=<42 42>
     margin=<ltrb 0 0 0 0>
     outlineSettings {
      bUseBrushTransparency=false
      roundingType=HalfHeightRadius
      width=0
      color {colorUseRule=UseColor_Specified, specifiedColor=<rgba 0 0 0 0>}
      cornerRadii=<0 0 0 0>
     }
     tintColor {colorUseRule=UseColor_Specified, specifiedColor=<rgba 1 1 1 1>}
     uVRegion {bIsValid=false, max=<0 0>, min=<0 0>}
    }}
   slot {class=HorizontalBoxSlot, properties {horizontalAlignment=HAlign_Center, verticalAlignment=VAlign_Center}}
  }
  CommonTextBlock_89 {
   widgetClassPath=&CommonTextBlock
   parent=HorizontalBox_85
   widget=CommonTextBlock_89
   properties {
    text=Called Functions
    textTransformPolicy=ToUpper
    colorAndOpacity {colorUseRule=UseColor_Specified, specifiedColor=<rgba 0.0044 0.3712 0.8308 1>}
    font {
     bForceMonospaced=false
     bMaterialIsStencil=false
     fontMaterial=null
     fontObject=&BurbankSmall
     letterSpacing=0
     monospacedWidth=1
     size=18
     skewAmount=0
     typefaceFontName=Black
     outlineSettings {
      bApplyOutlineToDropShadows=false
      bMiteredCorners=false
      bSeparateFillAlpha=false
      outlineMaterial=null
      outlineSize=0
      outlineColor=<rgba 0 0 0 1>
     }
    }
    shadowOffset=<0 0>
   }
   slot {class=HorizontalBoxSlot, properties {verticalAlignment=VAlign_Center, padding=<ltrb 14 0 0 0>, size {sizeRule=Fill, value=1}}}
  }
  Header_Events_Overlay {widgetClassPath=&Overlay, parent=Switcher, widget=Header_Events_Overlay, slot {class=WidgetSwitcherSlot}}
  Image_5 {
   widgetClassPath=&Image
   parent=Header_Events_Overlay
   widget=Image_5
   properties {
    visibility=Collapsed
    brush {
     bIsDynamicallyLoaded=false
     drawAs=Image
     imageType=NoImage
     mirroring=NoMirror
     resourceName=None
     resourceObject=null
     tiling=NoTile
     imageSize=<32 2>
     margin=<ltrb 0 0 0 0>
     outlineSettings {
      bUseBrushTransparency=false
      roundingType=HalfHeightRadius
      width=0
      color {colorUseRule=UseColor_Specified, specifiedColor=<rgba 0 0 0 0>}
      cornerRadii=<0 0 0 0>
     }
     tintColor {colorUseRule=UseColor_Specified, specifiedColor=<rgba 1 1 1 1>}
     uVRegion {bIsValid=false, max=<0 0>, min=<0 0>}
    }
    colorAndOpacity=<rgba 0.0024 0.016 0.117 1>
   }
   slot {class=OverlaySlot, properties {horizontalAlignment=HAlign_Fill, verticalAlignment=VAlign_Fill}}
  }
  HorizontalBox {
   widgetClassPath=&HorizontalBox
   parent=Header_Events_Overlay
   widget=HorizontalBox
   slot {class=OverlaySlot, properties {horizontalAlignment=HAlign_Fill, verticalAlignment=VAlign_Bottom, padding=<ltrb 12 32 12 12>}}
  }
  Image_1 {
   widgetClassPath=&Image
   parent=HorizontalBox
   widget=Image_1
   properties {brush {
     bIsDynamicallyLoaded=false
     drawAs=Image
     imageType=NoImage
     mirroring=NoMirror
     resourceName=None
     resourceObject=/Game/Athena/UI/InGame/Creative/PropertyEditor/Art/MI_UI_Binding_EventNormal
     tiling=NoTile
     imageSize=<42 42>
     margin=<ltrb 0 0 0 0>
     outlineSettings {
      bUseBrushTransparency=false
      roundingType=HalfHeightRadius
      width=0
      color {colorUseRule=UseColor_Specified, specifiedColor=<rgba 0 0 0 0>}
      cornerRadii=<0 0 0 0>
     }
     tintColor {colorUseRule=UseColor_Specified, specifiedColor=<rgba 1 1 1 1>}
     uVRegion {bIsValid=false, max=<0 0>, min=<0 0>}
    }}
   slot {class=HorizontalBoxSlot, properties {horizontalAlignment=HAlign_Center, verticalAlignment=VAlign_Center}}
  }
  CommonTextBlock {
   widgetClassPath=&CommonTextBlock
   parent=HorizontalBox
   widget=CommonTextBlock
   properties {
    text=OutGoing Events
    textTransformPolicy=ToUpper
    colorAndOpacity {colorUseRule=UseColor_Specified, specifiedColor=<rgba 0.8308 0.1651 0 1>}
    font {
     bForceMonospaced=false
     bMaterialIsStencil=false
     fontMaterial=null
     fontObject=&BurbankSmall
     letterSpacing=0
     monospacedWidth=1
     size=18
     skewAmount=0
     typefaceFontName=Black
     outlineSettings {
      bApplyOutlineToDropShadows=false
      bMiteredCorners=false
      bSeparateFillAlpha=false
      outlineMaterial=null
      outlineSize=0
      outlineColor=<rgba 0 0 0 1>
     }
    }
    shadowOffset=<0 0>
   }
   slot {class=HorizontalBoxSlot, properties {verticalAlignment=VAlign_Center, padding=<ltrb 14 0 0 0>, size {sizeRule=Fill, value=1}}}
  }
  Header_Title {widgetClassPath=&Overlay, parent=Switcher, widget=Header_Title, slot {class=WidgetSwitcherSlot}}
  Image_2 {
   widgetClassPath=&Image
   parent=Header_Title
   widget=Image_2
   properties {
    visibility=Collapsed
    brush {
     bIsDynamicallyLoaded=false
     drawAs=Image
     imageType=NoImage
     mirroring=NoMirror
     resourceName=None
     resourceObject=null
     tiling=NoTile
     imageSize=<32 2>
     margin=<ltrb 0 0 0 0>
     outlineSettings {
      bUseBrushTransparency=false
      roundingType=HalfHeightRadius
      width=0
      color {colorUseRule=UseColor_Specified, specifiedColor=<rgba 0 0 0 0>}
      cornerRadii=<0 0 0 0>
     }
     tintColor {colorUseRule=UseColor_Specified, specifiedColor=<rgba 1 1 1 1>}
     uVRegion {bIsValid=false, max=<0 0>, min=<0 0>}
    }
    colorAndOpacity=<rgba 0.0024 0.016 0.117 1>
   }
   slot {class=OverlaySlot, properties {horizontalAlignment=HAlign_Fill, verticalAlignment=VAlign_Fill}}
  }
  VerticalBox_0 {
   widgetClassPath=&VerticalBox
   parent=Header_Title
   widget=VerticalBox_0
   slot {class=OverlaySlot, properties {horizontalAlignment=HAlign_Fill, verticalAlignment=VAlign_Center, padding=<ltrb 14 4 4 4>}}
  }
  HorizontalBox_1 {
   widgetClassPath=&HorizontalBox
   parent=VerticalBox_0
   widget=HorizontalBox_1
   slot {class=VerticalBoxSlot, properties {padding=<ltrb 0 0 0 2>}}
  }
  Name_Text {
   widgetClassPath=&CommonTextBlock
   bIsVariable=true
   parent=HorizontalBox_1
   widget=Name_Text
   properties {
    text="[Name]"
    textTransformPolicy=ToUpper
    font {
     bForceMonospaced=false
     bMaterialIsStencil=false
     fontMaterial=null
     fontObject=&BurbankSmall
     letterSpacing=0
     monospacedWidth=1
     size=18
     skewAmount=0
     typefaceFontName=Black
     outlineSettings {
      bApplyOutlineToDropShadows=false
      bMiteredCorners=false
      bSeparateFillAlpha=false
      outlineMaterial=null
      outlineSize=0
      outlineColor=<rgba 0 0 0 1>
     }
    }
    shadowOffset=<0 0>
   }
   slot {class=HorizontalBoxSlot, properties {verticalAlignment=VAlign_Center, padding=<ltrb 4 0 0 0>, size {sizeRule=Fill, value=1}}}
  }
  Image_197 {
   widgetClassPath=&Image
   parent=VerticalBox_0
   widget=Image_197
   properties {
    brush {
     bIsDynamicallyLoaded=false
     drawAs=Image
     imageType=NoImage
     mirroring=NoMirror
     resourceName=None
     resourceObject=null
     tiling=NoTile
     imageSize=<32 3>
     margin=<ltrb 0 0 0 0>
     outlineSettings {
      bUseBrushTransparency=false
      roundingType=HalfHeightRadius
      width=0
      color {colorUseRule=UseColor_Specified, specifiedColor=<rgba 0 0 0 0>}
      cornerRadii=<0 0 0 0>
     }
     tintColor {colorUseRule=UseColor_Specified, specifiedColor=<rgba 1 1 1 1>}
     uVRegion {bIsValid=false, max=<0 0>, min=<0 0>}
    }
    colorAndOpacity=<rgba 1 1 1 0.2>
   }
   slot {class=VerticalBoxSlot}
  }
  Image_Backing {
   widgetClassPath=&Image
   bIsVariable=true
   parent=Overlay_0
   widget=Image_Backing
   properties {
    visibility=Collapsed
    brush {
     bIsDynamicallyLoaded=false
     drawAs=RoundedBox
     imageType=NoImage
     mirroring=NoMirror
     resourceName=None
     resourceObject=null
     tiling=NoTile
     imageSize=<16 2>
     margin=<ltrb 0 0 0 0>
     outlineSettings {
      bUseBrushTransparency=false
      roundingType=FixedRadius
      width=2
      color {colorUseRule=UseColor_Specified, specifiedColor=<rgba 0 0.0908 0.9911 1>}
      cornerRadii=<0 0 0 0>
     }
     tintColor {colorUseRule=UseColor_Specified, specifiedColor=<rgba 0.005 0.264 1 0.097>}
     uVRegion {bIsValid=false, max=<0 0>, min=<0 0>}
    }
   }
   slot {class=OverlaySlot, properties {horizontalAlignment=HAlign_Fill, verticalAlignment=VAlign_Fill}}
  }
  Image_Backing_Selected {
   widgetClassPath=&Image
   bIsVariable=true
   parent=Overlay_0
   widget=Image_Backing_Selected
   properties {
    visibility=Collapsed
    brush {
     bIsDynamicallyLoaded=false
     drawAs=RoundedBox
     imageType=NoImage
     mirroring=NoMirror
     resourceName=None
     resourceObject=null
     tiling=NoTile
     imageSize=<16 2>
     margin=<ltrb 0 0 0 0>
     outlineSettings {
      bUseBrushTransparency=false
      roundingType=FixedRadius
      width=2
      color {colorUseRule=UseColor_Specified, specifiedColor=<rgba 0 0.0908 0.9911 1>}
      cornerRadii=<0 0 0 0>
     }
     tintColor {colorUseRule=UseColor_Specified, specifiedColor=<rgba 0.9792 0.9846 1 0.097>}
     uVRegion {bIsValid=false, max=<0 0>, min=<0 0>}
    }
   }
   slot {class=OverlaySlot, properties {horizontalAlignment=HAlign_Fill, verticalAlignment=VAlign_Fill}}
  }
 ]
}
