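# Widget Blueprint for the Creative My Island event browser, listing devices and their event and function bindings, with teleport and traversal to linked devices.
# The native Fort Creative Device Browser serves searched, filtered, sorted pages; scrolling near the end requests more.
# Device selection requests subscriptions, which populate either the main list or the traversal dialog's list.
# Timers debounce search and throttle requests, and focus helpers handle gamepad navigation.
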WidgetBlueprint /Game/Athena/UI/InGame/Creative/MyIsland/Cr_EventBrowser
graphs [
 DebugPrintNewItem {nodes [
   n1 = Inputs!() -> (execute, InputPin:Event Browser Subscription List Entry Obj Object Reference)
   > n2 = Branch(Condition=NOT Boolean(A=Get DebugPrintNewItemEnabled() -> (DebugPrintNewItemEnabled:bool)))
   Outputs!(then:Exec=[@n5.then, @n2.then])
   n3 = Get bIsHeader!(self:Event Browser Subscription List Entry Obj Object Reference=@n1) -> (bIsHeader:bool)
   n4 = Get DeviceTrackingGUID!(self:Event Browser Subscription List Entry Obj Object Reference=@n1) -> (DeviceTrackingGUID:int=0)
   n5 = PrintString!(
    execute=@n2.else
    InString=Append!(
     A="CREATE SET STATE: DeviceTrackingGUID["
     B="To String (Integer)"(InInt=@n4)
     C:String="] DisplayName["
     D:String=Get DisplayName!(self:Event Browser Subscription List Entry Obj Object Reference=@n1) -> (DisplayName:String)
     E:String="] IsHeader["
     F:String="To String (Boolean)"(InBool=@n3)
     G:String="]"
    ) -> (ReturnValue)
    bPrintToScreen=Get DebugPrintNewItemPrint to Screen() -> (DebugPrintNewItemPrint to Screen:bool)
    bPrintToLog
    TextColor
    Duration
    Key
   ) -> (then)
  ]}
 DebugPrintSubscription {nodes [
   n1 = Inputs!() -> (execute, CreativeDeviceSubscription:"Creative Device Subscription Structure (by ref)")
   n2 = Break Creative Device Subscription!(CreativeDeviceSubscription:"Creative Device Subscription Structure (by ref)"=@n1) -> (
    DestinationDeviceGUID:int
    bIsEvent:bool=~
    SourceSubscriptionDisplayName:Text
    DestinationDeviceDisplayName:Text
    DestinationSubscriptionDisplayName:Text
    ActorPtr:Actor Soft Object Reference
   ) {is=title K2Node_BreakStruct}
   > PrintString(InString=Append!(
     A="To String (Integer)"(InInt=@n2)
     B="] "
     C=SelectString(A="EVENT ", B="FUNCTION ", bPickA=@n2.bIsEvent)
     D="To String (Text)"(InText=@n2.SourceSubscriptionDisplayName)
     E=" - "
     F="To String (Text)"(InText=@n2.DestinationDeviceDisplayName)
     G=" - "
     H:String="To String (Text)"(InText=@n2.DestinationSubscriptionDisplayName)
    ) -> (ReturnValue))
   > Outputs()
  ]}
 DebugPrintSubscriptionEntry {nodes [
   n1 = DebugPrintSubscriptionEntry!() -> (then, ListEntry:Object Reference, State:String) {is=title K2Node_FunctionEntry}
   > n2 = Cast To EventBrowserSubscriptionListEntryObj!(execute, Object=@n1) -> (then, CastFailed, AsEvent Browser Subscription List Entry Obj:Event Browser Subscription List Entry Obj Object Reference) {is=title K2Node_DynamicCast}
   > n3 = Branch(Condition=Get bIsHeader(self=@n2))
   > PrintString(InString=Append!(A=Get State() -> (State:String), B=: On Header) -> (ReturnValue), bPrintToScreen=false)
   n4 = Break Creative Device Subscription(CreativeDeviceSubscription=Get Subscription!(self:Event Browser Subscription List Entry Obj Object Reference=@n2) -> (Subscription:Creative Device Subscription Structure)) -> (bIsEvent=~)
   PrintString(
    execute=@n3.else
    InString=Append!(
     A=Get State()
     B=" GUID["
     C:String="To String (Integer)"(InInt=@n4)
     D:String="] Destination["
     E:String="To String (Text)"(InText=@n4.DestinationDeviceDisplayName)
     F:String="] Subscription [ "
     G:String="To String (Text)"(InText=@n4.DestinationSubscriptionDisplayName)
     H:String="]"
    ) -> (ReturnValue)
    bPrintToScreen=false
   )
  ]}
 EventGraph {nodes [
   "On Clicked (Button_Select)"!() -> (OutputDelegate, then, Button:Common Button Base Object Reference) {is=title K2Node_ComponentBoundEvent}
   > n1 = Branch(Condition=HasUserFocus!(
     self:Widget Object Reference=Get ListView_Devices() -> (ListView_Devices:Common List View Object Reference)
     PlayerController:Player Controller Object Reference=GetOwningPlayer()
    ) -> (ReturnValue:bool))
   RequestSubscriptionsForSelectedDevice(execute=[@n23.MouseAndKeyboard, @n23.Touch, @n1.then])
   Handle Subscription Selected(execute=@n1.else)
   n2 = "On Visibility Changed (Blade_SortAndFilter)"!() -> (OutputDelegate, then, InVisibility:ESlateVisibility Enum=Visible) {is=title K2Node_ComponentBoundEvent}
   > SetVisibility(self=Get Button_HitBlockerLeft() -> (Button_HitBlockerLeft:Fort Text Button Object Reference), InVisibility=@n2.InVisibility)
   > SetVisibility(self=Get Button_HitBlockerRight() -> (Button_HitBlockerRight:Fort Text Button Object Reference), InVisibility=@n2.InVisibility)
   > SetVisibility(self=Get Blade_BackgroundBlur() -> (Blade_BackgroundBlur:Background Blur Object Reference), InVisibility=@n2.InVisibility)
   n3 = "On Button Clicked (Button_TextSearch)"!() -> (OutputDelegate, then, Button:Common Button Legacy Object Reference) {is=title K2Node_ComponentBoundEvent}
   n4 = Cast To FortPlayerControllerGameplay!(execute=[@n3.then, @n35.then], Object=GetOwningPlayer()) -> (then, CastFailed, AsFort Player Controller Gameplay:Fort Player Controller Gameplay Object Reference) {is=title K2Node_DynamicCast}
   n5 = Get EditableText_Search() -> (EditableText_Search:WBP UIKit Filtered Input Field Object Reference)
   > n6 = Branch(Condition=OR Boolean(
     A=HasUserFocus(self=@n5, PlayerController=@n4)
     B=HasUserFocusedDescendants!(self:Widget Object Reference=@n5, PlayerController:Player Controller Object Reference=@n4) -> (ReturnValue:bool)
    ))
   > CenterWidget!(execute, ToCenter:Widget Object Reference=Get ListView_Devices()) -> (then)
   SetInputFieldText!(
    execute=@n6.else
    self:WBP UIKit Filtered Input Field Object Reference=Get EditableText_Search()
    Text:Text=GetEmptyText!() -> (ReturnValue:Text)
   ) -> (then)
   n7 = Get EditableText_Search()
   > CenterWidget(ToCenter=@n7)
   > SetTextFocus(self:WBP UIKit Filtered Input Field Object Reference=@n7)
   n8 = "On Button Clicked (Button_HitBlockerRight)"!() -> (OutputDelegate, then, Button:Common Button Legacy Object Reference) {is=title K2Node_ComponentBoundEvent}
   CloseBlade(
    execute=[@n8.then, @n9.then, @n10.then]
    self:Athena Customization Picker Sort And Filter Blade Object Reference=Get Blade_SortAndFilter() -> (Blade_SortAndFilter:Athena Customization Picker Sort And Filter Blade Object Reference)
   )
   n9 = "On Clicked (Button_CloseSortAndFilterBlade)"!() -> (OutputDelegate, then, Button:Common Button Base Object Reference) {is=title K2Node_ComponentBoundEvent}
   "On Hold Completed (Button_Teleport)"() {is=title K2Node_ComponentBoundEvent}
   > Teleport!(execute, self, ListItem:Object Reference) -> (then)
   n10 = "On Button Clicked (Button_HitBlockerLeft)"!() -> (OutputDelegate, then, Button:Common Button Legacy Object Reference) {is=title K2Node_ComponentBoundEvent}
   "On Clicked (Button_Back)"!() -> (OutputDelegate, then, Button:Common Button Base Object Reference) {is=title K2Node_ComponentBoundEvent}
   > n11 = Branch(Condition=IsTraversalDialogOpen!(self:Fort Creative Device Browser Object Reference) -> (ReturnValue:bool))
   > CloseTraversalDialog(self:Fort Creative Device Browser Object Reference)
   > SetFocus(self=Get ListView_Subscriptions() -> (ListView_Subscriptions:Fort Creative Device Browser List View Object Reference))
   n12 = Branch(execute=@n11.else, Condition=HasKeyboardFocus!(self:Widget Object Reference=Get EditableText_Search()) -> (ReturnValue:bool))
   n13 = Get ListView_Devices()
   > SetFocus(self=@n13)
   > SetSelectedIndex!(execute, self:List View Object Reference=@n13, Index:int=0) -> (then)
   DeactivateWidget(execute=@n12.else)
   n14 = "On List View Scrolled (ListView_Devices)"!() -> (OutputDelegate, then, ItemOffset:float, DistanceRemaining:float) {is=title K2Node_ComponentBoundEvent}
   > Set IsScrollingDown(IsScrollingDown:bool="float > float"(A=@n14.ItemOffset, B=Get PreviousScrollItemOffset() -> (PreviousScrollItemOffset:double))) -> (Output_Get:bool)
   > Set PreviousScrollItemOffset(PreviousScrollItemOffset:double=@n14.ItemOffset) -> (Output_Get:double)
   n15 = Get isWaitDirty_Devices() -> (isWaitDirty_Devices:bool)
   > Branch(Condition=AND Boolean(A=NOT Boolean(A=@n15), B=Get IsScrollingDown() -> (IsScrollingDown:bool)))
   > n16 = ShouldRequestNewPage!(
    execute
    self
    NumItems:int=GetNumItems!(self:List View Object Reference=Get ListView_Devices()) -> (ReturnValue:int=0)
    ItemOffset:float=@n14.ItemOffset
    DistanceRemaining:float=@n14.DistanceRemaining
   ) -> (then, ShouldDoRequest:bool)
   > Branch(Condition=@n16)
   > Set isWaitDirty_Devices(isWaitDirty_Devices:bool=true) -> (Output_Get:bool)
   > n17 = RequestPage!(
    execute
    self:Fort Creative Device Browser Object Reference
    bAddRequestedPageToCurrentPage:bool=true
    SearchString:String=Get CurrentSearchString() -> (CurrentSearchString:String)
    Filter:"EDeviceTrackingFilterType Enum (by ref)"=Get CurrentFilter() -> (CurrentFilter:EDeviceTrackingFilterType Enum=Undefined)
    SortToUse:"EDeviceTrackingSortingType Enum (by ref)"=Get CurrentSort() -> (CurrentSort:EDeviceTrackingSortingType Enum=Undefined)
   ) -> (then, ReturnValue:bool)
   > n18 = Branch(Condition=@n17)
   > SetVisibility(
    self=Get Image_Spinner_DevicesList() -> (Image_Spinner_DevicesList:Image Object Reference)
    InVisibility=Get ImageSpinnerDevicesList_EnabledVisibility() -> (ImageSpinnerDevicesList_EnabledVisibility:ESlateVisibility Enum=Visible)
   )
   n19 = Get CurrentFilter() -> (CurrentFilter=Undefined)
   > n20 = PrintString(
    InString=Append!(
     A=Getting more data...
     B=": "
     C="To String (Boolean)"(InBool=@n17)
     D=" FILTER: "
     E=Enum to String!(Enumerator:EDeviceTrackingFilterType Enum=@n19) -> (ReturnValue:String)
    ) -> (ReturnValue)
    TextColor=(R=1,G=0.5167,B=0.8157,A=1)
    Duration=3
   )
   Set Timer by Event(
    execute=[@n20.then, @n21.then]
    Delegate=Create Event()
    Time=GetRequestDelay!(self:Fort Creative Device Browser Object Reference) -> (ReturnValue:float)
    bLooping=false
   )
   n21 = PrintString(execute=@n18.else, InString=Linked-Scroll: Request returned false)
   n22 = "On Item Selection Changed (ListView_Devices)"!() -> (OutputDelegate, then, Item:Object Reference, bIsSelected:bool) {is=title K2Node_ComponentBoundEvent}
   > Branch(Condition=@n22.bIsSelected)
   > Set SelectedRowItem_ListViewDevices(SelectedRowItem_ListViewDevices:Object Reference=@n22.Item) -> (Output_Get:Object Reference)
   > n23 = Switch on ECommonInputType!(execute, Selection:ECommonInputType Enum=GetCurrentInputType(self=Get CommonInputSubsystem())) -> (MouseAndKeyboard:Exec, Gamepad:Exec, Touch:Exec, Count:Exec) {is=title K2Node_SwitchEnum}
   n24 = "On Text Changed (EditableText_Search)"!() -> (OutputDelegate, then, Text:Text) {is=title K2Node_ComponentBoundEvent}
   > Set TempSearchText(TempSearchText:Text=@n24.Text) -> (Output_Get:Text)
   > n25 = GetCurrentText!(execute, self:WBP UIKit Filtered Input Field Object Reference=Get EditableText_Search()) -> (then, Text:Text)
   > n26 = Branch(Condition=TextIsEmpty(InText=@n25))
   n27 = Get CommonInputSubsystem()
   Branch(execute=@n26.else, Condition="Equal (Enum)"(A=GetCurrentInputType(self=@n27), B=MouseAndKeyboard))
   > Set Timer by Event(
    Delegate=Create Event()
    Time=GetSearchInputDelay!(self:Fort Creative Device Browser Object Reference) -> (ReturnValue:float)
    bLooping=false
   )
   "On Text Committed (EditableText_Search)"!() -> (OutputDelegate, then, CommitMethod:ETextCommit Enum=Default) {is=title K2Node_ComponentBoundEvent}
   > n28 = GetCurrentText(self=Get EditableText_Search())
   > SearchChanged!(execute, self, InText:"Text (by ref)"=@n28) -> (then)
   EventOnRequestDelayElapsed() {is=title K2Node_CustomEvent}
   > Set isWaitDirty_Bindings(isWaitDirty_Bindings:bool=false) -> (Output_Get:bool)
   OnCommitSearchTextAfterDelay() {is=title K2Node_CustomEvent}
   > n29 = GetCurrentText(self=Get EditableText_Search())
   n30 = Get TempSearchText() -> (TempSearchText:Text)
   > Branch(Condition="Equal, Case Insensitive (Text)"!(A:"Text (by ref)"=@n29, B:"Text (by ref)"=@n30) -> (ReturnValue:bool))
   > SearchChanged(InText=@n30)
   n31 = OnInputMethodChanged_Event!() -> (OutputDelegate, then, bNewInputType:ECommonInputType Enum) {is=title K2Node_CustomEvent}
   n32 = Get Blade_SortAndFilter()
   > n33 = Branch(Condition=IsBladeOpen!(self:Athena Customization Picker Sort And Filter Blade Object Reference=@n32) -> (ReturnValue:bool))
   > n34 = CloseBlade(self=@n32)
   ResetListViewSelection!(execute=[@n34.then, @n33.else], self, InputType:ECommonInputType Enum=@n31.bNewInputType) -> (then)
   CloseUI() {is=title K2Node_CustomEvent}
   > RequestCloseUI(self:Fort Creative Device Browser Object Reference)
   ScrollRequestDelayReady() {is=title K2Node_CustomEvent}
   > Set isWaitDirty_Devices(isWaitDirty_Devices=false)
   Event OnTeleporting() {is=title K2Node_Event}
   > RequestCloseUI()
   Event On Activated()
   > Set isWaitDirty_Devices(isWaitDirty_Devices=false)
   > Set isWaitDirty_Bindings(isWaitDirty_Bindings=false)
   > Scan()
   > SetListViewFocus!(execute, self, InListViewRef:Common List View Object Reference=Get ListView_Devices(), ListViewItem:Object Reference) -> (then)
   n35 = Event HandleButtonTextSearchClicked() {is=title K2Node_Event}
   n36 = Event OnDeviceDetailsInfoReceived!() -> (OutputDelegate, then, SelectedDeviceInfo:"Creative Device Info Structure (by ref)") {is=title K2Node_Event}
   > n37 = Set Selected Device Info(Selected Device Info:Creative Device Info Structure=@n36.SelectedDeviceInfo) -> (Output_Get:Creative Device Info Structure)
   n38 = Branch(execute=[@n41.then, @n37.then], Condition=Get isWaitDirty_Bindings() -> (isWaitDirty_Bindings:bool))
   Set isWaitDirty_Bindings(execute=@n38.else, isWaitDirty_Bindings=true)
   > HideCommonTextMessage()
   > PopulateBindingsForDevice!(
    execute
    self
    TargetDevice:Creative Device Info Structure=Get Selected Device Info() -> (Selected Device Info:Creative Device Info Structure)
   ) -> (then)
   > Set Timer by Event(Delegate=Create Event(), Time=GetRequestDelay(), bLooping=false)
   Event OnSelectedDeviceInfoFailed() {is=title K2Node_Event}
   > ResetBindingsPanel()
   > ShowCommonTextMessage!(execute, self, Message:Text="LOCTEXT(\"Failed to select device\")") -> (then)
   > SetVisibility(
    self=[
     Get Image_Spinner_Subscriptions() -> (Image_Spinner_Subscriptions:Image Object Reference)
     Get Image_Spinner_Traversal_Subscriptions() -> (Image_Spinner_Traversal_Subscriptions:Image Object Reference)
    ]
    InVisibility=Get ImageSpinnerSubscriptionsList_DisabledVisibility() -> (ImageSpinnerSubscriptionsList_DisabledVisibility:ESlateVisibility Enum=Visible)
   )
   Event OnListViewRefreshNeeded() {is=title K2Node_Event}
   > Scan()
   Event Construct()
   > n39 = Sequence()
   > Assign On Input Method Changed!(execute:Exec, self:Common Input Object Reference=Get CommonInputSubsystem(), Delegate:"Delegate (by ref)"=Create Event()) -> (then:Exec) {is=title K2Node_AssignDelegate}
   SetVisibility(
    execute=@n39.then_1
    self=Get Image_Spinner_DevicesList()
    InVisibility=Get ImageSpinnerDevicesList_DisabledVisibility() -> (ImageSpinnerDevicesList_DisabledVisibility:ESlateVisibility Enum=Visible)
   )
   > SetVisibility(
    self=[Get Image_Spinner_Subscriptions(), Get Image_Spinner_Traversal_Subscriptions()]
    InVisibility=Get ImageSpinnerSubscriptionsList_DisabledVisibility() -> (ImageSpinnerSubscriptionsList_DisabledVisibility=Visible)
   )
   n40 = Event OnSelectedDeviceNoSubscriptions!() -> (OutputDelegate, then, SelectedDeviceInfo:"Creative Device Info Structure (by ref)") {is=title K2Node_Event}
   > n41 = Set Selected Device Info(Selected Device Info=@n40.SelectedDeviceInfo)
   Event OnDeviceInfosPageUpdated() {is=title K2Node_Event}
   > DebugPrintCreativeDeviceInfo(self:Fort Creative Device Browser Object Reference)
   > n42 = GetIndexOfSelectionInListView!(execute, self, ListView:List View Object Reference=Get ListView_Devices()) -> (then, Index:int=0)
   > Set IndexBeforePaging(IndexBeforePaging:int=@n42) -> (Output_Get:int=0)
   > PopulateDevices()
   > SetVisibility(
    self=Get Image_Spinner_DevicesList()
    InVisibility=Get ImageSpinnerDevicesList_DisabledVisibility() -> (ImageSpinnerDevicesList_DisabledVisibility=Visible)
   )
   n43 = Get ListView_Devices()
   > n44 = Branch(Condition="integer > integer"(A=GetNumItems(self=@n43) -> (ReturnValue=0)))
   > HideCommonTextMessage()
   > SetSelectionInListViewByIndex!(execute, self, ListView:List View Object Reference=Get ListView_Devices(), Index:int=Get IndexBeforePaging() -> (IndexBeforePaging:int=0)) -> (then)
   n45 = Switch on EDeviceTrackingFilterType!(execute=@n44.else, Selection:EDeviceTrackingFilterType Enum=Get CurrentFilter() -> (CurrentFilter=Undefined)) -> (
    Undefined:Exec
    Unfiltered:Exec
    WithEventsOrFunctions:Exec
    WithFunctions:Exec
    WithEvents:Exec
    WithOnlyFunctions:Exec
    WithOnlyEvents:Exec
    WithBindings:Exec
    Favorites:Exec
   ) {is=title K2Node_SwitchEnum}
   ShowCommonTextMessage(
    execute=[
     @n45.Favorites
     @n45.Unfiltered
     @n45.WithBindings
     @n45.WithEvents
     @n45.WithEventsOrFunctions
     @n45.WithFunctions
     @n45.WithOnlyEvents
     @n45.WithOnlyFunctions
    ]
    Message="LOCTEXT(\"No devices found.\")"
   )
   ShowCommonTextMessage(Message="LOCTEXT(\"No devices with bindings  found.\")")
  ]}
 GetDeviceGUIDFromListViewObject {nodes [
   n1 = GetDeviceGUIDFromListViewObject!() -> (then, ListViewObject:Object Reference) {is=title K2Node_FunctionEntry}
   > n2 = Sequence!(execute) -> (then_0, then_1, then_2:Exec)
   > n3 = Cast To EventBrowserSubscriptionListEntryObj(Object=@n1)
   > Return Node!(execute, ReturnValue:int=Break Creative Device Subscription(CreativeDeviceSubscription=Get Subscription(self=@n3)) -> (bIsEvent=~))
   n4 = Cast To Obj_DvBrowserDeviceListItem!(execute=@n2.then_1, Object=@n1) -> (then, CastFailed, AsObj Dv Browser Device List Item:Obj Dv Browser Device List Item Object Reference) {is=title K2Node_DynamicCast}
   > Return Node(ReturnValue=Get DeviceTrackingGUID(self:Obj Dv Browser Device List Item Object Reference=@n4) -> (DeviceTrackingGUID=0))
   Return Node(execute=@n2.then_2, ReturnValue=-1)
  ]}
 GetIndexOfSelectionInListView {nodes [
   n1 = GetIndexOfSelectionInListView!() -> (then, ListView:List View Object Reference) {is=title K2Node_FunctionEntry}
   n2 = GetIndexForItem!(
    self:List View Object Reference=@n1
    Item:Object Reference=Get Selected Item!(self:List View Object Reference=@n1) -> (ReturnValue:Object Reference) {is=title K2Node_CallFunction}
   ) -> (ReturnValue:int=0)
   > Return Node!(execute, Index:int=SelectInt(A=0, B=@n2, bPickA="integer < integer"(A=@n2)))
  ]}
 GetTeleportActorFromListViewObject {nodes [
   n1 = GetTeleportActorFromListViewObject!() -> (then, ListViewObject:Object Reference) {is=title K2Node_FunctionEntry}
   > n2 = GetTeleportTargetForListViewItem!(execute, self, Object:Object Reference=@n1) -> (then, TeleportTarget:Actor Object Reference)
   > Return Node!(execute, ReturnValue:Actor Object Reference=@n2)
  ]}
 GetTeleportTargetForListViewItem {nodes [
   n1 = GetTeleportTargetForListViewItem!() -> (then, Object:Object Reference) {is=title K2Node_FunctionEntry}
   > n2 = Branch(Condition=IsTeleportEnabled!(self:Fort Creative Device Browser Object Reference) -> (ReturnValue:bool))
   > n3 = Cast To EventBrowserSubscriptionListEntryObj(Object=@n1)
   n4 = Get Subscription(self=@n3)
   > Return Node!(
    execute
    TeleportTarget:Actor Object Reference=Resolve Soft Reference!(Input:Actor Soft Object Reference=Break Creative Device Subscription(CreativeDeviceSubscription=@n4) -> (bIsEvent=~).ActorPtr) -> (Output:Actor Object Reference)
   )
   n5 = Cast To Obj_DvBrowserDeviceListItem(execute=@n3.CastFailed, Object=@n1)
   > Return Node(TeleportTarget=Resolve Soft Reference(Input=Get DeviceActor!(self:Obj Dv Browser Device List Item Object Reference=@n5) -> (DeviceActor:Actor Soft Object Reference)))
   Return Node(execute=[@n5.CastFailed, @n2.else])
  ]}
 Handle Subscription Selected {nodes [
   Handle Subscription Selected() {is=title K2Node_FunctionEntry}
   > n1 = Sequence!(execute) -> (then_0, then_1)
   RequestSubscriptionsForSelectedDevice(execute=@n1.then_1)
   > Return Node!(execute)
   n2 = IsBrowsingDevices!(self:Fort Creative Device Browser Object Reference) -> (ReturnValue:bool)
   Branch(
    execute=@n1.then_0
    Condition=AND Boolean(A=IsTraversalEnabled!(self:Fort Creative Device Browser Object Reference) -> (ReturnValue:bool), B=NOT Boolean(A=@n2))
   )
   > n3 = Branch(Condition=IsTraversalDialogOpen())
   > n4 = Set CurrentSubscriptionsListView(CurrentSubscriptionsListView:Fort Creative Device Browser List View Object Reference=Get ListView_Traversal_Subscriptions() -> (ListView_Traversal_Subscriptions:Fort Creative Device Browser List View Object Reference)) -> (Output_Get:Fort Creative Device Browser List View Object Reference)
   n5 = Get CurrentSubscriptionsListView() -> (CurrentSubscriptionsListView:Fort Creative Device Browser List View Object Reference)
   Branch(
    execute=[@n4.then, @n6.then]
    Condition="integer > integer"(A=Get Num Items Selected!(self:List View Object Reference=@n5) -> (ReturnValue:int=0) {is=title K2Node_CallFunction})
   )
   > Traverse to Device!(execute, self, SelectedDevice:Object Reference=Get Selected Item(self=Get CurrentSubscriptionsListView())) -> (then)
   > Return Node()
   n6 = Set CurrentSubscriptionsListView(execute=@n3.else, CurrentSubscriptionsListView=Get ListView_Subscriptions())
  ]}
 HandleNavFromDeviceListToSubscriptions {nodes [
   HandleNavFromDeviceListToSubscriptions!() -> (then, Navigation:EUINavigation Enum) {is=title K2Node_FunctionEntry}
   > n1 = Sequence()
   n2 = Get ListView_Subscriptions()
   > Branch(Condition="integer > integer"(A=GetNumItems(self=@n2) -> (ReturnValue=0), B=1))
   > Set LinkedDv_LastItem(LinkedDv_LastItem:Object Reference=Get Selected Item(self=Get ListView_Devices())) -> (Output_Get:Object Reference)
   > SetListViewFocus(InListViewRef=Get ListView_Subscriptions(), ListViewItem=Get Bindings_LastItem() -> (Bindings_LastItem:Object Reference))
   Return Node!(execute=@n1.then_1, ReturnValue:Widget Object Reference)
  ]}
 HandleNavFromSubscriptionsToDeviceList {nodes [
   HandleNavFromSubscriptionsToDeviceList!() -> (then, Navigation:EUINavigation Enum) {is=title K2Node_FunctionEntry}
   > Set Bindings_LastItem(Bindings_LastItem:Object Reference=Get Selected Item(self=Get ListView_Subscriptions())) -> (Output_Get:Object Reference)
   > SetListViewFocus(InListViewRef=Get ListView_Devices(), ListViewItem=Get LinkedDv_LastItem() -> (LinkedDv_LastItem:Object Reference))
   > Set SelectedRowItemBindings(SelectedRowItemBindings:Object Reference) -> (Output_Get:Object Reference)
   > Set HoveredRowItemBindings(HoveredRowItemBindings:Object Reference) -> (Output_Get:Object Reference)
   > Return Node()
  ]}
 HandleTextSearchNav {nodes [
   HandleTextSearchNav!() -> (then, Navigation:EUINavigation Enum) {is=title K2Node_FunctionEntry}
   > SetListViewFocus(InListViewRef=Get ListView_Devices())
   > Return Node()
  ]}
 HideCommonTextMessage {nodes [
   HideCommonTextMessage() {is=title K2Node_FunctionEntry}
   > SetActiveWidget(
    self=Get ListViewHeaderSwitch() -> (ListViewHeaderSwitch:Common Visibility Switcher Object Reference)
    Widget=Get ListViewHeader_Legend() -> (ListViewHeader_Legend:Size Box Object Reference)
   )
   > "SetText (Text)"(self=Get CommonTextBlock_Status() -> (CommonTextBlock_Status:Common Text Object Reference))
  ]}
 PopulateBindingsForDevice {nodes [
   PopulateBindingsForDevice!() -> (then, TargetDevice:Creative Device Info Structure) {is=title K2Node_FunctionEntry}
   > n1 = Sequence!(execute) -> (then_0, then_1, then_2:Exec, then_3:Exec, then_4:Exec, then_5:Exec)
   n2 = Get OutgoingEvents() -> (OutgoingEvents:Array of Creative Device Subscription Structures)
   Branch(execute=@n1.then_1, Condition="integer > integer"(A=Length(TargetArray:Array of Creative Device Subscription Structures=@n2)))
   > n3 = CreateSubscriptionListHeader!(execute, self:Fort Creative Device Browser Object Reference, Outer:Object Reference=GetOwningPlayer(), bIsEvent:bool=true) -> (then, ReturnValue:Event Browser Subscription List Entry Obj Object Reference)
   > AddItem!(
    execute
    self:List View Object Reference=Get ListViewToPopulate() -> (ListViewToPopulate:Common List View Object Reference)
    Item:Object Reference=@n3
   ) -> (then)
   > n4 = For Each Loop(Array:Array of Creative Device Subscription Structures=Get OutgoingEvents()) -> (Array Element:Creative Device Subscription Structure)
   > n5 = CreateSubscriptionListItem!(
    execute
    self:Fort Creative Device Browser Object Reference
    Outer:Object Reference=GetOwningPlayer()
    Subscription:"Creative Device Subscription Structure (by ref)"=@n4
   ) -> (then, ReturnValue:Event Browser Subscription List Entry Obj Object Reference)
   > AddItem(self=Get ListViewToPopulate(), Item=@n5)
   n6 = Get CalledFunctions() -> (CalledFunctions:Array of Creative Device Subscription Structures)
   Branch(execute=@n1.then_2, Condition="integer > integer"(A=Length(TargetArray=@n6)))
   > n7 = CreateSubscriptionListHeader(Outer=GetOwningPlayer(), bIsEvent=false)
   > AddItem(self=Get ListViewToPopulate(), Item=@n7)
   > n8 = For Each Loop(Array=Get CalledFunctions())
   > n9 = CreateSubscriptionListItem(Outer=GetOwningPlayer(), Subscription=@n8)
   > AddItem(self=Get ListViewToPopulate(), Item=@n9)
   n10 = Branch(execute=@n1.then_3, Condition=Get HasSubscriptions() -> (HasSubscriptions:bool))
   n11 = CreateSubscriptionListMessage!(
    execute=@n10.else
    self:Fort Creative Device Browser Object Reference
    Outer:Object Reference=GetOwningPlayer()
    Message:String=No events are bound
   ) -> (then, ReturnValue:Event Browser Subscription List Entry Obj Object Reference)
   > AddItem(self=Get ListViewToPopulate(), Item=@n11)
   n12 = Break Creative Device Info!(CreativeDeviceInfo:"Creative Device Info Structure (by ref)"=Get TargetDevice() -> (TargetDevice:Creative Device Info Structure)) -> (
    DeviceTrackingGUID:int
    ActorPtr:Actor Soft Object Reference
    DisplayName:String
    OutgoingEvents:Array of Creative Device Subscription Structures
    CalledFunctions:Array of Creative Device Subscription Structures
    NumBoundEvents:int
    NumBoundFunctions:int
   ) {is=title K2Node_BreakStruct}
   Set OutgoingEvents(execute=@n1.then_0, OutgoingEvents:Array of Creative Device Subscription Structures=@n12.OutgoingEvents) -> (Output_Get:Array of Creative Device Subscription Structures)
   > Set CalledFunctions(CalledFunctions:Array of Creative Device Subscription Structures=@n12.CalledFunctions) -> (Output_Get:Array of Creative Device Subscription Structures)
   > Set HasSubscriptions(HasSubscriptions:bool="integer > integer"(A=int + int(A=@n12.NumBoundEvents, B=@n12.NumBoundFunctions))) -> (Output_Get:bool)
   > n13 = Set ListViewToPopulate(ListViewToPopulate:Common List View Object Reference=Select(
     Option 0:Fort Creative Device Browser List View Object Reference=Get ListView_Subscriptions()
     Option 1:Fort Creative Device Browser List View Object Reference=Get ListView_Traversal_Subscriptions()
     Index=IsTraversalDialogOpen()
    ) -> (ReturnValue:Fort Creative Device Browser List View Object Reference)) -> (Output_Get:Common List View Object Reference)
   > ClearListItems(self=@n13)
   SetVisibility(
    execute=@n1.then_4
    self=[Get Image_Spinner_Traversal_Subscriptions(), Get Image_Spinner_Subscriptions()]
    InVisibility=Get ImageSpinnerSubscriptionsList_DisabledVisibility() -> (ImageSpinnerSubscriptionsList_DisabledVisibility=Visible)
   )
   Branch(execute=@n1.then_5, Condition=AND Boolean(A=IsTraversalDialogOpen(), B=Get HasSubscriptions()))
   > SetListViewFocus(InListViewRef=Get ListViewToPopulate())
  ]}
 PopulateDevices {nodes [
   PopulateDevices() {is=title K2Node_FunctionEntry}
   > ClearListItems(self=Get ListView_Devices())
   > n1 = For Each Loop(Array:Array of Integers=Get CreativeDeviceInfo() -> (CreativeDeviceInfo:Array of Integers)) -> (Array Element:int)
   n2 = Find(
    TargetMap:Map of Integers to Creative Device Info Structures=Get CreativeDeviceInfoMap() -> (CreativeDeviceInfoMap:Map of Integers to Creative Device Info Structures)
    Key:"Integer (by ref)"=@n1
   ) -> (Value:Creative Device Info Structure)
   > Branch(Condition=@n2.ReturnValue)
   n3 = Break Creative Device Info(CreativeDeviceInfo=@n2)
   > n4 = Construct Object from Class!(
    execute
    Class=&/Obj_DvBrowserDeviceListItem.Obj_DvBrowserDeviceListItem_C
    self=GetOwningPlayer()
    DisplayName:String=@n3.DisplayName
    DeviceTrackingGUID:int=@n3
    NumEvents:int=@n3.NumBoundEvents
    NumFunctions:int=@n3.NumBoundFunctions
    DeviceActor:Actor Soft Object Reference=@n3.ActorPtr
   ) -> (then, ReturnValue:Obj Dv Browser Device List Item Object Reference)
   > AddItem(self=Get ListView_Devices(), Item=@n4)
  ]}
 PrintDebugData {nodes [
   PrintDebugData() {is=title K2Node_FunctionEntry}
   > n1 = For Each Loop(Array=Get CreativeDeviceInfo())
   n2 = Find(TargetMap=Get CreativeDeviceInfoMap(), Key=@n1)
   > Set DebugDeviceInfo(DebugDeviceInfo:Creative Device Info Structure=@n2) -> (Output_Get:Creative Device Info Structure)
   > n3 = Branch(Condition=@n2.ReturnValue)
   n4 = Break Creative Device Info(CreativeDeviceInfo=Get DebugDeviceInfo() -> (DebugDeviceInfo:Creative Device Info Structure))
   > PrintString(
    InString=Append!(A="\r\nACTOR:\r\n", B=@n4.DisplayName, C="-----------------------------\r\n") -> (ReturnValue)
    TextColor=(R=0,G=1,B=0.3936,A=1)
   )
   > n5 = For Each Loop(Array:Array of Creative Device Subscription Structures=@n4.OutgoingEvents) -> (Array Element:Creative Device Subscription Structure)
   n6 = For Each Loop(Exec=@n5.Completed, Array=@n4.CalledFunctions)
   > Debug Print Subscription!(execute:Exec, CreativeDeviceSubscription:"Creative Device Subscription Structure (by ref)"=@n6) -> (then:Exec) {is=title K2Node_MacroInstance}
   Debug Print Subscription(execute=@n5.LoopBody, CreativeDeviceSubscription=@n5)
   PrintString(execute=@n3.else, InString=Append(A="Device GUID[", B="To String (Integer)"(InInt=@n1), C="] not found in map"))
  ]}
 RequestSubscriptionsForDeviceGUID {nodes [
   n1 = RequestSubscriptionsForDeviceGUID!() -> (then, DeviceGUID:int) {is=title K2Node_FunctionEntry}
   > n2 = RequestSelectDevice!(execute, self:Fort Creative Device Browser Object Reference, DeviceTrackingGUIDToAssign:int=@n1) -> (then, ReturnValue:bool)
   > n3 = Branch(Condition=@n2)
   n4 = IsTraversalDialogOpen()
   > ClearListItems(self=Select(
     Option 0:Common List View Object Reference=Get ListView_Subscriptions()
     Option 1:Common List View Object Reference=Get ListView_Traversal_Subscriptions()
     Index=@n4
    ) -> (ReturnValue:Common List View Object Reference))
   > n5 = SetVisibility(
    self=Select(
     Option 0:Image Object Reference=Get Image_Spinner_Subscriptions()
     Option 1:Image Object Reference=Get Image_Spinner_Traversal_Subscriptions()
     Index=@n4
    ) -> (ReturnValue:Image Object Reference)
    InVisibility=Get ImageSpinnerSubscriptionsList_EnabledVisibility() -> (ImageSpinnerSubscriptionsList_EnabledVisibility:ESlateVisibility Enum=Visible)
   )
   HideCommonTextMessage(execute=[@n5.then, @n3.else]) {is=title K2Node_CallFunction}
  ]}
 RequestSubscriptionsForSelectedDevice {nodes [
   RequestSubscriptionsForSelectedDevice() {is=title K2Node_FunctionEntry}
   > n1 = Cast To Obj_DvBrowserDeviceListItem(Object=Get SelectedRowItem_ListViewDevices() -> (SelectedRowItem_ListViewDevices:Object Reference))
   > RequestSubscriptionsForDeviceGUID!(execute, self, DeviceGUID:int=Get DeviceTrackingGUID(self=@n1) -> (DeviceTrackingGUID=0)) -> (then) {is=title K2Node_CallFunction}
  ]}
 ResetBindingListViewSelection {nodes [
   n1 = ResetBindingListViewSelection!() -> (then, InputType:ECommonInputType Enum) {is=title K2Node_FunctionEntry}
   > Branch(Condition="Equal (Enum)"(A=@n1, B=Gamepad))
   n2 = Get ListView_Subscriptions()
   > SetListViewFocus(InListViewRef=@n2, ListViewItem=Get Selected Item(self=@n2))
  ]}
 ResetBindingsPanel {nodes [
   ResetBindingsPanel() {is=title K2Node_FunctionEntry}
   > ClearListItems(self=Get ListView_Subscriptions())
  ]}
 ResetListViewSelection {nodes [
   n1 = ResetListViewSelection!() -> (then, InputType:ECommonInputType Enum) {is=title K2Node_FunctionEntry}
   > Branch(Condition="Equal (Enum)"(A=@n1, B=Gamepad))
   n2 = Get ListView_Subscriptions()
   n3 = GetNumItems(self=@n2) -> (ReturnValue=0)
   n4 = Select(
    Option 0:Common List View Object Reference=Select(Option 0=Get ListView_Devices(), Option 1=@n2, Index="integer > integer"(A=@n3, B=1))
    Option 1:Common List View Object Reference=Get ListView_Traversal_Subscriptions()
    Index=IsTraversalDialogOpen()
   ) -> (ReturnValue:Common List View Object Reference)
   > SetListViewFocus(InListViewRef=@n4, ListViewItem=Get Selected Item(self=@n4))
  ]}
 Scan {nodes [
   Scan() {is=title K2Node_FunctionEntry}
   > SetVisibility(
    self=Get Image_Spinner_DevicesList()
    InVisibility=Get ImageSpinnerDevicesList_EnabledVisibility() -> (ImageSpinnerDevicesList_EnabledVisibility=Visible)
   )
   > ShowCommonTextMessage(Message="LOCTEXT(\"Scanning for Devices\")")
   > ClearListItems(self=Get ListView_Devices())
   > ResetBindingsPanel() {is=title K2Node_CallFunction}
   > Set PreviousScrollItemOffset(PreviousScrollItemOffset=0)
   > RequestPage(
    bAddRequestedPageToCurrentPage=false
    SearchString=Get CurrentSearchString()
    Filter=Get CurrentFilter() -> (CurrentFilter=Undefined)
    SortToUse=Get CurrentSort() -> (CurrentSort=Undefined)
   )
  ]}
 SearchChanged {nodes [
   n1 = SearchChanged!() -> (then, InText:"Text (by ref)") {is=title K2Node_FunctionEntry}
   n2 = Get CurrentSearchString()
   > Branch(Condition="Not Equal Exactly (Text)"!(A:"Text (by ref)"="To Text (String)"(InString=@n2), B:"Text (by ref)"=@n1) -> (ReturnValue:bool))
   > SetSearchString!(execute, self:Fort Creative Device Browser Object Reference, SearchString:String="To String (Text)"(InText=@n1)) -> (then)
   > Scan() {is=title K2Node_CallFunction}
  ]}
 SetListViewFocus {nodes [
   n1 = SetListViewFocus!() -> (then, InListViewRef:Common List View Object Reference, ListViewItem:Object Reference) {is=title K2Node_FunctionEntry}
   > SetFocus(self=@n1)
   > n2 = Is Valid(InputObject=@n1.ListViewItem)
   n3 = Get InListViewRef() -> (InListViewRef:Common List View Object Reference)
   SetSelectedIndex(execute=@n2.Is Not Valid, self=@n3, Index=1)
   Set Selected Item!(execute=@n2.Is Valid, self:List View Object Reference=@n3, Item:Object Reference=@n1.ListViewItem) -> (then) {is=title K2Node_CallFunction}
  ]}
 SetSelectionInListViewByIndex {nodes [
   n1 = SetSelectionInListViewByIndex!() -> (then, ListView:List View Object Reference, Index:int) {is=title K2Node_FunctionEntry}
   > SetFocus(self=@n1)
   n2 = GetItemAt!(self:List View Object Reference=@n1, Index:int=@n1.Index) -> (ReturnValue:Object Reference)
   > Set Selected Item(self=@n1, Item=@n2)
   > n3 = Switch on ECommonInputType(Selection=GetCurrentInputType(self=Get CommonInputSubsystem()))
   Scroll Item Into View!(execute=@n3.Gamepad, self:List View Object Reference=@n1, Item:Object Reference=@n2) -> (then)
  ]}
 ShouldRequestNewPage {nodes [
   n1 = ShouldRequestNewPage!() -> (then, NumItems:int, ItemOffset:float=~, DistanceRemaining:float=~) {is=title K2Node_FunctionEntry}
   n2 = GetScrollOffsetToRequestNewPage!(self:Fort Creative Device Browser Object Reference) -> (ReturnValue:float)
   n3 = Get DataPageSize() -> (DataPageSize:int=0)
   > n4 = Branch(Condition="float < float"(A=float - float(A:int=@n1, B=@n1.ItemOffset), B=float * float(A=@n2, B:int=@n3) {is=title K2Node_PromotableOperator}))
   n5 = Get DistanceRemaining() -> (DistanceRemaining:float)
   n6 = Get ItemOffset() -> (ItemOffset:float)
   n7 = Get NumItems() -> (NumItems:int=0)
   > PrintString(InString=Append!(
     A="ShouldRequestNewPage: Offset = "
     B="To String (Float)"(InDouble=@n6)
     C=" DistanceRemaining = "
     D:String="To String (Float)"(InDouble=@n5)
     E:String=" NumItems = "
     F:String="To String (Integer)"(InInt=@n7)
    ) -> (ReturnValue))
   > Return Node!(execute, ShouldDoRequest:bool=true)
   Return Node(execute=@n4.else, ShouldDoRequest=false)
  ]}
 ShowCommonTextMessage {nodes [
   n1 = ShowCommonTextMessage!() -> (then, Message:Text) {is=title K2Node_FunctionEntry}
   > "SetText (Text)"(self=Get CommonTextBlock_Status(), InText=@n1)
   > SetActiveWidget(self=Get ListViewHeaderSwitch(), Widget=Get ListViewHeader_Message() -> (ListViewHeader_Message:Size Box Object Reference))
  ]}
 Teleport {nodes [
   n1 = Teleport!() -> (then, ListItem:Object Reference) {is=title K2Node_FunctionEntry}
   > RequestTeleportByListItem!(execute, self:Fort Creative Device Browser Object Reference, ListItem:Object Reference=@n1) -> (then)
  ]}
 Traverse to Device {nodes [
   Traverse to Device!() -> (then, SelectedDevice:Object Reference) {is=title K2Node_FunctionEntry}
   > n1 = Sequence!(execute) -> (then_0, then_1, then_2, then_3)
   ClearListItems(execute=@n1.then_2, self=Get ListView_Traversal_Device() -> (ListView_Traversal_Device:Common List View Object Reference))
   n2 = Break Creative Device Cached!(CreativeDeviceCached:"Creative Device Cached Structure (by ref)"=Get TraverseDeviceInfo() -> (TraverseDeviceInfo:Creative Device Cached Structure)) -> (DisplayName:Text, NumBoundEvents:int, NumBoundFunctions:int, ActorPtr:Actor Soft Object Reference) {is=title K2Node_BreakStruct}
   > n3 = Construct Object from Class(
    Class=&Obj_DvBrowserDeviceListItem_C
    self=GetOwningPlayer()
    DisplayName="To String (Text)"(InText=@n2)
    DeviceTrackingGUID=Get DestinationDeviceGUID() -> (DestinationDeviceGUID:int=0)
    NumEvents=@n2.NumBoundEvents
    NumFunctions=@n2.NumBoundFunctions
    DeviceActor=@n2.ActorPtr
   )
   > AddItem(self=Get ListView_Traversal_Device(), Item=@n3)
   n4 = Cast To EventBrowserSubscriptionListEntryObj(execute=@n1.then_0, Object=Get SelectedDevice() -> (SelectedDevice:Object Reference))
   n5 = Break Creative Device Subscription(CreativeDeviceSubscription=Get Subscription(self=@n4)) -> (bIsEvent=~)
   n6 = Find(
    TargetMap:Map of Integers to Creative Device Cached Structures=Get KnownDevicesMap() -> (KnownDevicesMap:Map of Integers to Creative Device Cached Structures)
    Key=@n5
   ) -> (Value:Creative Device Cached Structure)
   > n7 = Branch(Condition=AND Boolean(
     A=@n6.ReturnValue
     B=IsTraversalAllowed!(self:Fort Creative Device Browser Object Reference, DeviceTrackingGUID:int=@n5) -> (ReturnValue:bool)
    ))
   > Set DestinationDeviceGUID(DestinationDeviceGUID:int=@n5) -> (Output_Get:int=0)
   > Set TraverseDeviceInfo(TraverseDeviceInfo:Creative Device Cached Structure=@n6) -> (Output_Get:Creative Device Cached Structure)
   n8 = PrintString(execute=@n7.else, InString=Traverse Failed to find GUID in info map)
   Return Node!(execute=[@n8.then, @n9.then])
   n9 = PrintString(execute=@n4.CastFailed, InString=Traverse Cast Failed)
   OpenTraversalDialog(execute=@n1.then_1, self:Fort Creative Device Browser Object Reference)
   RequestSubscriptionsForDeviceGUID(execute=@n1.then_3, DeviceGUID=Get DestinationDeviceGUID() -> (DestinationDeviceGUID=0))
  ]}
]
variables [
 Bindings_LastItem {type=Object, default=null}
 ClassDisplayNameMap {default {}}
 DataPageSize {type=integer, default=100}
 DebugPrintNewItemEnabled {type=boolean}
 DebugPrintNewItemPrint to Screen {type=boolean}
 HoveredRowItemBindings {type=Object, default=null}
 ImageSpinnerDevicesList_DisabledVisibility {type=ESlateVisibility, default=Hidden}
 ImageSpinnerDevicesList_EnabledVisibility {type=ESlateVisibility, default=SelfHitTestInvisible}
 ImageSpinnerSubscriptionsList_DisabledVisibility {type=ESlateVisibility, default=Collapsed}
 ImageSpinnerSubscriptionsList_EnabledVisibility {type=ESlateVisibility, default=SelfHitTestInvisible}
 IndexBeforePaging {type=integer}
 Invalid Class List {type="array<Class>", default []}
 IsScrollingDown {type=boolean}
 isWaitDirty_Bindings {type=boolean}
 isWaitDirty_Devices {type=boolean}
 LinkedDv_LastItem {type=Object, default=null}
 PreviousScrollItemOffset {type=number}
 Selected Device Info {type=CreativeDeviceInfo, default {deviceTrackingGUId=-1, displayName="", numBoundEvents=0, numBoundFunctions=0}}
 SelectedRowItem_ListViewDevices {type=Object, default=null}
 SelectedRowItemBindings {type=Object, default=null}
 TempSearchText {type=string, default=""}
]
widgets {
 info {inheritedWidgetCount=14, parentClass=/Script/FortniteUI.FortCreativeDeviceBrowser, rootWidgetClass=&Overlay, widgetCount=119}
 widgets [
  Overlay_EventBrowser {widgetClassPath=&Overlay, widget=Overlay_EventBrowser}
  Lightbox {
   widgetClassPath=/Game/UI/Foundation/Lightbox/Lightbox
   parent=Overlay_EventBrowser
   widget=Lightbox
   properties {dimAmount=0.9, showWindowControlButtons=false}
   slot {class=OverlaySlot, properties {horizontalAlignment=HAlign_Fill, verticalAlignment=VAlign_Fill}}
  }
  MainHB {
   widgetClassPath=&HorizontalBox
   parent=Overlay_EventBrowser
   widget=MainHB
   slot {class=OverlaySlot, properties {horizontalAlignment=HAlign_Fill, verticalAlignment=VAlign_Fill}}
  }
  Overlay_192 {widgetClassPath=&Overlay, parent=MainHB, widget=Overlay_192, slot {class=HorizontalBoxSlot}}
  SafeZone_EventBrowser {
   widgetClassPath=&SafeZone
   parent=Overlay_192
   widget=SafeZone_EventBrowser
   properties {padBottom=false, padRight=false}
   slot {class=OverlaySlot, properties {horizontalAlignment=HAlign_Fill, verticalAlignment=VAlign_Fill}}
  }
  Overlay_542 {widgetClassPath=&Overlay, parent=SafeZone_EventBrowser, widget=Overlay_542, slot {class=SafeZoneSlot}}
  Image_Backing {
   widgetClassPath=&Image
   parent=Overlay_542
   widget=Image_Backing
   properties {
    visibility=SelfHitTestInvisible
    brush {
     bIsDynamicallyLoaded=false
     drawAs=Image
     imageType=NoImage
     mirroring=NoMirror
     resourceName=None
     resourceObject=null
     tiling=NoTile
     imageSize=<32 32>
     margin=<ltrb 0 0 0 0>
     outlineSettings {
      bUseBrushTransparency=false
      roundingType=HalfHeightRadius
      width=0
      color {colorUseRule=UseColor_Specified, specifiedColor=<rgba 0 0 0 0>}
      cornerRadii=<0 0 0 0>
     }
     tintColor {colorUseRule=UseColor_Specified, specifiedColor=<rgba 0.0024 0.016 0.117 1>}
     uVRegion {bIsValid=false, max=<0 0>, min=<0 0>}
    }
   }
   slot {class=OverlaySlot, properties {horizontalAlignment=HAlign_Fill, verticalAlignment=VAlign_Fill}}
  }
  SafezonePaddingReductionCB {
   widgetClassPath=&CommonBorder
   parent=Overlay_542
   widget=SafezonePaddingReductionCB
   properties {bReducePaddingBySafezone=true, minimumPadding=<ltrb 0 -64 0 0>, padding=<ltrb 0 0 0 0>}
   slot {class=OverlaySlot, properties {horizontalAlignment=HAlign_Fill, verticalAlignment=VAlign_Fill, padding=<ltrb 0 64 0 80>}}
  }
  Overlay_3 {
   widgetClassPath=&Overlay
   parent=SafezonePaddingReductionCB
   widget=Overlay_3
   slot {class=BorderSlot, properties {padding=<ltrb 0 0 0 0>}}
  }
  SizeBox_EventBrowser {
   widgetClassPath=&SizeBox
   parent=Overlay_3
   widget=SizeBox_EventBrowser
   properties {bOverride_WidthOverride=true, widthOverride=775}
   slot {class=OverlaySlot, properties {horizontalAlignment=HAlign_Fill, verticalAlignment=VAlign_Fill}}
  }
  Overlay_0 {widgetClassPath=&Overlay, parent=SizeBox_EventBrowser, widget=Overlay_0, slot {class=SizeBoxSlot}}
  HorizontalBox_151 {
   widgetClassPath=&HorizontalBox
   parent=Overlay_0
   widget=HorizontalBox_151
   slot {class=OverlaySlot, properties {horizontalAlignment=HAlign_Fill, verticalAlignment=VAlign_Fill}}
  }
  SizeBox_LeftPanel {
   widgetClassPath=&SizeBox
   parent=HorizontalBox_151
   widget=SizeBox_LeftPanel
   properties {bOverride_WidthOverride=true, widthOverride=775}
   slot {class=HorizontalBoxSlot, properties {horizontalAlignment=HAlign_Left}}
  }
  VerticalBox_0 {
   widgetClassPath=&VerticalBox
   parent=SizeBox_LeftPanel
   widget=VerticalBox_0
   slot {class=SizeBoxSlot, properties {horizontalAlignment=HAlign_Left}}
  }
  SizeBox_Title {
   widgetClassPath=&SizeBox
   parent=VerticalBox_0
   widget=SizeBox_Title
   properties {bOverride_HeightOverride=true, heightOverride=100}
   slot {class=VerticalBoxSlot}
  }
  Overlay_2 {widgetClassPath=&Overlay, parent=SizeBox_Title, widget=Overlay_2, slot {class=SizeBoxSlot}}
  Border_39 {
   widgetClassPath=&Border
   parent=Overlay_2
   widget=Border_39
   properties {brushColor=<rgba 0.0024 0.016 0.117 1>, padding=<ltrb 0 0 0 0>}
   slot {class=OverlaySlot, properties {horizontalAlignment=HAlign_Fill, verticalAlignment=VAlign_Fill}}
  }
  CommonTextBlock_2 {
   widgetClassPath=&CommonTextBlock
   parent=Overlay_2
   widget=CommonTextBlock_2
   properties {
    justification=Center
    text=EVENT BROWSER
    font {
     bForceMonospaced=false
     bMaterialIsStencil=false
     fontMaterial=null
     fontObject=&Burbank
     letterSpacing=0
     monospacedWidth=1
     size=40
     skewAmount=0
     typefaceFontName=BurbankRegular-Black
     outlineSettings {
      bApplyOutlineToDropShadows=false
      bMiteredCorners=false
      bSeparateFillAlpha=false
      outlineMaterial=&M-FontShadow
      outlineSize=0
      outlineColor=<rgba 0 0 0 1>
     }
    }
    shadowOffset=<4.9266 5.2555>
   }
   slot {class=OverlaySlot, properties {verticalAlignment=VAlign_Bottom, padding=<ltrb 50 0 0 0>}}
  }
  HorizontalBox_Options {
   widgetClassPath=&HorizontalBox
   parent=VerticalBox_0
   widget=HorizontalBox_Options
   slot {class=VerticalBoxSlot, properties {horizontalAlignment=HAlign_Left, padding=<ltrb 0 10 0 10>}}
  }
  SizeBox_0 {
   widgetClassPath=&SizeBox
   parent=HorizontalBox_Options
   widget=SizeBox_0
   properties {bOverride_HeightOverride=true, heightOverride=32}
   slot {class=HorizontalBoxSlot, properties {padding=<ltrb 50 0 0 18>}}
  }
  HorizontalBox {
   widgetClassPath=&HorizontalBox
   parent=SizeBox_0
   widget=HorizontalBox
   properties {flowDirectionPreference=Culture}
   slot {class=SizeBoxSlot}
  }
  Border_0 {
   widgetClassPath=&Border
   parent=HorizontalBox
   widget=Border_0
   properties {
    verticalAlignment=VAlign_Center
    background {
     bIsDynamicallyLoaded=false
     drawAs=Box
     imageType=NoImage
     mirroring=NoMirror
     resourceName=None
     resourceObject=/Game/Athena/UI/Customization/Art/M_UI_SortAndFilter_Tag
     tiling=NoTile
     imageSize=<32 32>
     margin=<ltrb 0.5 0 0.5 0>
     outlineSettings {
      bUseBrushTransparency=false
      roundingType=HalfHeightRadius
      width=0
      color {colorUseRule=UseColor_Specified, specifiedColor=<rgba 0 0 0 0>}
      cornerRadii=<0 0 0 0>
     }
     tintColor {colorUseRule=UseColor_Specified, specifiedColor=<rgba 1 1 1 1>}
     uVRegion {bIsValid=false, max=<0 0>, min=<0 0>}
    }
    contentColorAndOpacity=<rgba 0.0044 0.0273 0.2502 1>
    padding=<ltrb 20 0 20 0>
   }
   slot {class=HorizontalBoxSlot}
  }
  Text_SortTag {
   widgetClassPath=&CommonTextBlock
   bInherited=true
   bIsVariable=true
   parent=Border_0
   widget=Text_SortTag
   properties {
    justification=Center
    mobileFontSizeMultiplier=1.3
    scrollStyle=&TextScrollStyle-Base_C
    text="Sort: [PH]"
    textTransformPolicy=ToUpper
    colorAndOpacity {colorUseRule=UseColor_Specified, specifiedColor=<rgba 1 0 1 1>}
    font {
     bForceMonospaced=false
     bMaterialIsStencil=false
     fontMaterial=null
     fontObject=&BurbankSmall
     letterSpacing=0
     monospacedWidth=1
     size=12
     skewAmount=0
     typefaceFontName=Black
     outlineSettings {
      bApplyOutlineToDropShadows=false
      bMiteredCorners=false
      bSeparateFillAlpha=false
      outlineMaterial=null
      outlineSize=0
      outlineColor=<rgba 0 0 0 1>
     }
    }
    renderTransform {angle=0, scale=<1 1>, shear=<-8 0>, translation=<0 0>}
    shadowOffset=<0 0>
   }
   slot {class=BorderSlot, properties {verticalAlignment=VAlign_Center, padding=<ltrb 20 0 20 0>}}
  }
  Border {
   widgetClassPath=&Border
   parent=HorizontalBox
   widget=Border
   properties {
    verticalAlignment=VAlign_Center
    background {
     bIsDynamicallyLoaded=false
     drawAs=Box
     imageType=NoImage
     mirroring=NoMirror
     resourceName=None
     resourceObject=&M_UI_SortAndFilter_Tag
     tiling=NoTile
     imageSize=<32 32>
     margin=<ltrb 0.5 0 0.5 0>
     outlineSettings {
      bUseBrushTransparency=false
      roundingType=HalfHeightRadius
      width=0
      color {colorUseRule=UseColor_Specified, specifiedColor=<rgba 0 0 0 0>}
      cornerRadii=<0 0 0 0>
     }
     tintColor {colorUseRule=UseColor_Specified, specifiedColor=<rgba 1 1 1 1>}
     uVRegion {bIsValid=false, max=<0 0>, min=<0 0>}
    }
    contentColorAndOpacity=<rgba 0.0044 0.0273 0.2502 1>
    padding=<ltrb 20 0 20 0>
   }
   slot {class=HorizontalBoxSlot}
  }
  Text_FilterTag {
   widgetClassPath=&CommonTextBlock
   bInherited=true
   bIsVariable=true
   parent=Border
   widget=Text_FilterTag
   properties {
    justification=Center
    mobileFontSizeMultiplier=1.3
    scrollStyle=&TextScrollStyle-Base_C
    text="Filter: [PH]"
    textTransformPolicy=ToUpper
    colorAndOpacity {colorUseRule=UseColor_Specified, specifiedColor=<rgba 1 0 1 1>}
    font {
     bForceMonospaced=false
     bMaterialIsStencil=false
     fontMaterial=null
     fontObject=&BurbankSmall
     letterSpacing=0
     monospacedWidth=1
     size=12
     skewAmount=0
     typefaceFontName=Black
     outlineSettings {
      bApplyOutlineToDropShadows=false
      bMiteredCorners=false
      bSeparateFillAlpha=false
      outlineMaterial=null
      outlineSize=0
      outlineColor=<rgba 0 0 0 1>
     }
    }
    renderTransform {angle=0, scale=<1 1>, shear=<-8 0>, translation=<0 0>}
    shadowOffset=<0 0>
   }
   slot {class=BorderSlot, properties {verticalAlignment=VAlign_Center, padding=<ltrb 20 0 20 0>}}
  }
  HorizontalBox_Search {
   widgetClassPath=&HorizontalBox
   parent=VerticalBox_0
   widget=HorizontalBox_Search
   slot {class=VerticalBoxSlot, properties {padding=<ltrb 50 10 0 0>}}
  }
  Border_2 {
   widgetClassPath=&Border
   parent=HorizontalBox_Search
   widget=Border_2
   properties {
    background {
     bIsDynamicallyLoaded=false
     drawAs=Image
     imageType=NoImage
     mirroring=NoMirror
     resourceName=None
     resourceObject=null
     tiling=NoTile
     imageSize=<32 32>
     margin=<ltrb 0.5 0.5 0.5 0.5>
     outlineSettings {
      bUseBrushTransparency=false
      roundingType=HalfHeightRadius
      width=0
      color {colorUseRule=UseColor_Specified, specifiedColor=<rgba 0 0 0 0>}
      cornerRadii=<0 0 0 0>
     }
     tintColor {colorUseRule=UseColor_Specified, specifiedColor=<rgba 1 1 1 1>}
     uVRegion {bIsValid=false, max=<0 0>, min=<0 0>}
    }
    brushColor=<rgba 0.0024 0.0024 0.0999 0>
    padding=<ltrb 0 0 0 0>
   }
   slot {class=HorizontalBoxSlot, properties {padding=<ltrb 0 0 28 0>, size {sizeRule=Fill, value=1}}}
  }
  HorizontalBox_2 {
   widgetClassPath=&HorizontalBox
   parent=Border_2
   widget=HorizontalBox_2
   slot {class=BorderSlot, properties {padding=<ltrb 0 0 0 0>}}
  }
  CommonVisibilityWidget_InputAction {
   widgetClassPath=&CommonVisibilityWidgetLegacy
   parent=HorizontalBox_2
   widget=CommonVisibilityWidget_InputAction
   properties {
    bShowForAndroid=false
    bShowForIOS=false
    bShowForMac=false
    bShowForMouseAndKeyboard=false
    bShowForTouch=false
    horizontalAlignment=HAlign_Center
    verticalAlignment=VAlign_Center
    padding=<ltrb 0 0 8 0>
   }
   slot {class=HorizontalBoxSlot, properties {horizontalAlignment=HAlign_Right, verticalAlignment=VAlign_Center}}
  }
  Button_TextSearch {
   widgetClassPath=/FortUILibrary/Foundation/Buttons/PanelButton
   bInherited=true
   bIsVariable=true
   parent=CommonVisibilityWidget_InputAction
   widget=Button_TextSearch
   properties {
    bDisplayInputActionWhenNotInteractable=false
    bInteractableWhenSelected=true
    inputActionUseRimBrush=false
    inputActionVerticalAlignment=VAlign_Center
    navigation=&self:WidgetTree.Button_TextSearch.WidgetNavigation_0
    onButtonBaseClicked=(&self:WidgetTree.Button_TextSearch.HandleOnButtonClicked)
    onButtonBaseDoubleClicked=(&self:WidgetTree.Button_TextSearch.HandleOnButtonDoubleClicked)
    onButtonBaseHovered=(&self:WidgetTree.Button_TextSearch.HandleOnButtonHovered)
    onButtonBaseUnhovered=(&self:WidgetTree.Button_TextSearch.HandleOnButtonUnhovered)
    onSelectedChangedBase=(&self:WidgetTree.Button_TextSearch.HandleOnSelectedChanged)
    style=/FortUILibrary/Foundation/Buttons/ButtonStyles/Athena/ButtonStyle_TransparentList_Nothing.ButtonStyle_TransparentList_Nothing_C
    inputActionRenderTranslation=<0 0>
    triggeringInputAction {dataTable=&InputActionDataTable, rowName=GenericLeftTrigger}
   }
   slot {class=BorderSlot, properties {horizontalAlignment=HAlign_Center, verticalAlignment=VAlign_Center, padding=<ltrb 0 0 8 0>}}
  }
  FortMobileResize {
   widgetClassPath=&ScaleBox
   parent=HorizontalBox_2
   widget=FortMobileResize
   properties {stretch=None, stretchDirection=UpOnly}
   slot {class=HorizontalBoxSlot, properties {padding=<ltrb 0 4 4 4>, size {sizeRule=Fill, value=1}}}
  }
  EditableText_Search {
   widgetClassPath=/FortUILibrary/UIKit/InputField/WBP_UIKit_FilteredInputField
   bIsVariable=true
   parent=FortMobileResize
   widget=EditableText_Search
   properties {
    bIsFocusable=true
    entryCharLimit=100
    hideValidationInfo=true
    hintText=SEARCH
    isSearchField=true
    navigation=&self:WidgetTree.EditableText_Search.WidgetNavigation_0
    promptText=""
    showTextPrompt=false
    suggestionText=SEARCH
    useHintText=true
    virtualKeyboardTrigger=OnAllFocusEvents
   }
   slot {class=ScaleBoxSlot, properties {horizontalAlignment=HAlign_Fill, verticalAlignment=VAlign_Fill}}
  }
  Overlay {
   widgetClassPath=&Overlay
   parent=HorizontalBox_Search
   widget=Overlay
   slot {class=HorizontalBoxSlot, properties {padding=<ltrb 0 0 28 0>}}
  }
  MaxWidth_SB {
   widgetClassPath=&SizeBox
   parent=Overlay
   widget=MaxWidth_SB
   properties {bOverride_MaxDesiredWidth=true, bOverride_MinDesiredWidth=true, maxDesiredWidth=320, minDesiredWidth=200}
   slot {class=OverlaySlot, properties {horizontalAlignment=HAlign_Fill, verticalAlignment=VAlign_Center}}
  }
  Button_SortAndFilter {
   widgetClassPath=&WBP_UIKit_Button_Regular
   bInherited=true
   bIsVariable=true
   parent=MaxWidth_SB
   widget=Button_SortAndFilter
   properties {
    minHeight=90
    mobileMinHeight=70
    navigation=&self:WidgetTree.Button_SortAndFilter.WidgetNavigation_0
    overrideDefaultHeight=true
    text=Sort+Filter
    triggeringInputAction {dataTable=&InputActionDataTable, rowName=GenericRightTrigger}
   }
   slot {class=SizeBoxSlot}
  }
  SizeBox_DeviceList {
   widgetClassPath=&SizeBox
   parent=VerticalBox_0
   widget=SizeBox_DeviceList
   properties {bOverride_HeightOverride=true, heightOverride=690}
   slot {class=VerticalBoxSlot, properties {size {sizeRule=Fill, value=1}}}
  }
  HorizontalBox_Devices {
   widgetClassPath=&HorizontalBox
   parent=SizeBox_DeviceList
   widget=HorizontalBox_Devices
   properties {
    clipping=ClipToBounds
    navigation=&self:WidgetTree.HorizontalBox_Devices.WidgetNavigation_0
   }
   slot {class=SizeBoxSlot}
  }
  SizeBox_4 {
   widgetClassPath=&SizeBox
   parent=HorizontalBox_Devices
   widget=SizeBox_4
   properties {bOverride_WidthOverride=true, heightOverride=620, widthOverride=760}
   slot {class=HorizontalBoxSlot, properties {horizontalAlignment=HAlign_Left}}
  }
  Overlay_DevicesList {widgetClassPath=&Overlay, parent=SizeBox_4, widget=Overlay_DevicesList, slot {class=SizeBoxSlot}}
  Image_Backing_2 {
   widgetClassPath=&Image
   parent=Overlay_DevicesList
   widget=Image_Backing_2
   properties {
    visibility=SelfHitTestInvisible
    brush {
     bIsDynamicallyLoaded=false
     drawAs=Image
     imageType=NoImage
     mirroring=NoMirror
     resourceName=None
     resourceObject=null
     tiling=NoTile
     imageSize=<32 32>
     margin=<ltrb 0 0 0 0>
     outlineSettings {
      bUseBrushTransparency=false
      roundingType=HalfHeightRadius
      width=0
      color {colorUseRule=UseColor_Specified, specifiedColor=<rgba 0 0 0 0>}
      cornerRadii=<0 0 0 0>
     }
     tintColor {colorUseRule=UseColor_Specified, specifiedColor=<rgba 0.0024 0.016 0.117 1>}
     uVRegion {bIsValid=false, max=<0 0>, min=<0 0>}
    }
   }
   slot {class=OverlaySlot, properties {horizontalAlignment=HAlign_Fill, verticalAlignment=VAlign_Fill}}
  }
  VerticalBox_ListView {
   widgetClassPath=&VerticalBox
   parent=Overlay_DevicesList
   widget=VerticalBox_ListView
   slot {class=OverlaySlot, properties {horizontalAlignment=HAlign_Fill, verticalAlignment=VAlign_Fill, padding=<ltrb 50 0 12 0>}}
  }
  ListViewHeaderSwitch {
   widgetClassPath=&CommonVisibilitySwitcher
   bIsVariable=true
   parent=VerticalBox_ListView
   widget=ListViewHeaderSwitch
   slot {class=VerticalBoxSlot}
  }
  ListViewHeader_Legend {
   widgetClassPath=&SizeBox
   bIsVariable=true
   parent=ListViewHeaderSwitch
   widget=ListViewHeader_Legend
   properties {bOverride_HeightOverride=true, heightOverride=60, widthOverride=800}
   slot {class=CommonVisibilitySwitcherSlot}
  }
  HorizontalBox_HeaderDevices {
   widgetClassPath=&HorizontalBox
   parent=ListViewHeader_Legend
   widget=HorizontalBox_HeaderDevices
   properties {visibility=Visible}
   slot {class=SizeBoxSlot, properties {horizontalAlignment=HAlign_Right}}
  }
  Image_Events {
   widgetClassPath=&Image
   parent=HorizontalBox_HeaderDevices
   widget=Image_Events
   properties {brush {
     bIsDynamicallyLoaded=false
     drawAs=Image
     imageType=NoImage
     mirroring=NoMirror
     resourceName=None
     resourceObject=/Game/Athena/UI/InGame/Creative/PropertyEditor/Art/MI_UI_Binding_EventNormal
     tiling=NoTile
     imageSize=<42 42>
     margin=<ltrb 0 0 0 0>
     outlineSettings {
      bUseBrushTransparency=false
      roundingType=HalfHeightRadius
      width=0
      color {colorUseRule=UseColor_Specified, specifiedColor=<rgba 0 0 0 0>}
      cornerRadii=<0 0 0 0>
     }
     tintColor {colorUseRule=UseColor_Specified, specifiedColor=<rgba 1 1 1 1>}
     uVRegion {bIsValid=false, max=<0 0>, min=<0 0>}
    }}
   slot {class=HorizontalBoxSlot, properties {horizontalAlignment=HAlign_Center, verticalAlignment=VAlign_Center, padding=<ltrb 0 0 32 0>}}
  }
  Spacer_508 {
   widgetClassPath=&Spacer
   parent=HorizontalBox_HeaderDevices
   widget=Spacer_508
   properties {size=<30 1>}
   slot {class=HorizontalBoxSlot}
  }
  Image_Functions {
   widgetClassPath=&Image
   parent=HorizontalBox_HeaderDevices
   widget=Image_Functions
   properties {brush {
     bIsDynamicallyLoaded=false
     drawAs=Image
     imageType=NoImage
     mirroring=NoMirror
     resourceName=None
     resourceObject=/Game/Athena/UI/InGame/Creative/PropertyEditor/Art/MI_UI_Binding_FunctionNormal
     tiling=NoTile
     imageSize=<42 42>
     margin=<ltrb 0 0 0 0>
     outlineSettings {
      bUseBrushTransparency=false
      roundingType=HalfHeightRadius
      width=0
      color {colorUseRule=UseColor_Specified, specifiedColor=<rgba 0 0 0 0>}
      cornerRadii=<0 0 0 0>
     }
     tintColor {colorUseRule=UseColor_Specified, specifiedColor=<rgba 1 1 1 1>}
     uVRegion {bIsValid=false, max=<0 0>, min=<0 0>}
    }}
   slot {class=HorizontalBoxSlot, properties {horizontalAlignment=HAlign_Center, verticalAlignment=VAlign_Center, padding=<ltrb 0 0 18 0>}}
  }
  ListViewHeader_Message {
   widgetClassPath=&SizeBox
   bIsVariable=true
   parent=ListViewHeaderSwitch
   widget=ListViewHeader_Message
   properties {bOverride_HeightOverride=true, heightOverride=60, widthOverride=800}
   slot {class=CommonVisibilitySwitcherSlot}
  }
  CommonTextBlock_Status {
   widgetClassPath=&CommonTextBlock
   bIsVariable=true
   parent=ListViewHeader_Message
   widget=CommonTextBlock_Status
   properties {
    bAutoCollapseWithEmptyText=true
    justification=Center
    style=/FortUILibrary/Styles/Typography/Foundation/TextStyle_BurbankSmall_S_LightBLue.TextStyle_BurbankSmall_S_LightBLue_C
    text=No bound devices found
    colorAndOpacity {colorUseRule=UseColor_Specified, specifiedColor=<rgba 0.2051 0.6172 0.9473 1>}
    font {
     bForceMonospaced=false
     bMaterialIsStencil=false
     fontMaterial=null
     fontObject=&BurbankSmall
     letterSpacing=0
     monospacedWidth=1
     size=18
     skewAmount=0
     typefaceFontName=Black
     outlineSettings {
      bApplyOutlineToDropShadows=false
      bMiteredCorners=false
      bSeparateFillAlpha=false
      outlineMaterial=null
      outlineSize=0
      outlineColor=<rgba 0 0 0 1>
     }
    }
    shadowOffset=<0 0>
   }
   slot {class=SizeBoxSlot, properties {verticalAlignment=VAlign_Bottom}}
  }
  ListView_Devices {
   widgetClassPath=&CommonListView
   bInherited=true
   bIsVariable=true
   parent=VerticalBox_ListView
   widget=ListView_Devices
   properties {
    entryWidgetClass=&/Cr_DvBrowserDvEntry.Cr_DvBrowserDvEntry_C
    horizontalEntrySpacing=8
    navigation=&self:WidgetTree.ListView_Devices.WidgetNavigation_0
    numDesignerPreviewEntries=6
    verticalEntrySpacing=8
   }
   slot {class=VerticalBoxSlot, properties {padding=<ltrb 0 0 0 20>, size {sizeRule=Fill, value=1}}}
  }
  Image_Spinner_DevicesList {
   widgetClassPath=&Image
   bIsVariable=true
   parent=VerticalBox_ListView
   widget=Image_Spinner_DevicesList
   properties {
    visibility=Collapsed
    brush {
     bIsDynamicallyLoaded=false
     drawAs=Image
     imageType=NoImage
     mirroring=NoMirror
     resourceName=None
     resourceObject=/FortUILibrary/UIKit/Throbber/MI_UIKit_Color_Throbber
     tiling=NoTile
     imageSize=<48 48>
     margin=<ltrb 0 0 0 0>
     outlineSettings {
      bUseBrushTransparency=false
      roundingType=HalfHeightRadius
      width=0
      color {colorUseRule=UseColor_Specified, specifiedColor=<rgba 0 0 0 0>}
      cornerRadii=<0 0 0 0>
     }
     tintColor {colorUseRule=UseColor_Specified, specifiedColor=<rgba 1 1 1 1>}
     uVRegion {bIsValid=false, max=<0 0>, min=<0 0>}
    }
   }
   slot {class=VerticalBoxSlot, properties {horizontalAlignment=HAlign_Center, verticalAlignment=VAlign_Bottom, padding=<ltrb 0 0 0 16>}}
  }
  Button_HitBlockerLeft {
   widgetClassPath=/FortUILibrary/Foundation/Buttons/FortTextButton
   bIsVariable=true
   parent=Overlay_542
   widget=Button_HitBlockerLeft
   properties {
    bHideInputAction=true
    defaultText=""
    onButtonBaseClicked=(&self:WidgetTree.Button_HitBlockerLeft.HandleOnButtonClicked)
    onButtonBaseDoubleClicked=(&self:WidgetTree.Button_HitBlockerLeft.HandleOnButtonDoubleClicked)
    onButtonBaseHovered=(&self:WidgetTree.Button_HitBlockerLeft.HandleOnButtonHovered)
    onButtonBaseUnhovered=(&self:WidgetTree.Button_HitBlockerLeft.HandleOnButtonUnhovered)
    onSelectedChangedBase=(&self:WidgetTree.Button_HitBlockerLeft.HandleOnSelectedChanged)
    style=/FortUILibrary/Foundation/Buttons/ButtonStyles/ButtonStyle-FullyInvisible.ButtonStyle-FullyInvisible_C
   }
   slot {class=OverlaySlot, properties {horizontalAlignment=HAlign_Fill, verticalAlignment=VAlign_Fill}}
  }
  SafeZone_Right {
   widgetClassPath=&SafeZone
   parent=MainHB
   widget=SafeZone_Right
   properties {padBottom=false, padLeft=false}
   slot {class=HorizontalBoxSlot, properties {size {sizeRule=Fill, value=1}}}
  }
  Overlay_805 {widgetClassPath=&Overlay, parent=SafeZone_Right, widget=Overlay_805, slot {class=SafeZoneSlot}}
  SizeBox_RightPanel {
   widgetClassPath=&SizeBox
   parent=Overlay_805
   widget=SizeBox_RightPanel
   properties {bOverride_WidthOverride=true, widthOverride=1162}
   slot {class=OverlaySlot, properties {horizontalAlignment=HAlign_Fill, verticalAlignment=VAlign_Fill}}
  }
  Overlay_Bindings {widgetClassPath=&Overlay, parent=SizeBox_RightPanel, widget=Overlay_Bindings, slot {class=SizeBoxSlot}}
  Image_Backing_1 {
   widgetClassPath=&Image
   parent=Overlay_Bindings
   widget=Image_Backing_1
   properties {
    visibility=SelfHitTestInvisible
    brush {
     bIsDynamicallyLoaded=false
     drawAs=Image
     imageType=NoImage
     mirroring=NoMirror
     resourceName=None
     resourceObject=null
     tiling=NoTile
     imageSize=<32 32>
     margin=<ltrb 0 0 0 0>
     outlineSettings {
      bUseBrushTransparency=false
      roundingType=HalfHeightRadius
      width=0
      color {colorUseRule=UseColor_Specified, specifiedColor=<rgba 0 0 0 0>}
      cornerRadii=<0 0 0 0>
     }
     tintColor {colorUseRule=UseColor_Specified, specifiedColor=<rgba 0.0009 0.006 0.0395 1>}
     uVRegion {bIsValid=false, max=<0 0>, min=<0 0>}
    }
    colorAndOpacity=<rgba 1 1 1 0.8>
   }
   slot {class=OverlaySlot, properties {horizontalAlignment=HAlign_Fill, verticalAlignment=VAlign_Fill}}
  }
  Image_Spinner_Subscriptions {
   widgetClassPath=&Image
   bIsVariable=true
   parent=Overlay_Bindings
   widget=Image_Spinner_Subscriptions
   properties {
    visibility=Collapsed
    brush {
     bIsDynamicallyLoaded=false
     drawAs=Image
     imageType=NoImage
     mirroring=NoMirror
     resourceName=None
     resourceObject=/FortUILibrary/UIKit/Throbber/MI_UIKit_Monochrome_Throbber
     tiling=NoTile
     imageSize=<64 64>
     margin=<ltrb 0 0 0 0>
     outlineSettings {
      bUseBrushTransparency=false
      roundingType=HalfHeightRadius
      width=0
      color {colorUseRule=UseColor_Specified, specifiedColor=<rgba 0 0 0 0>}
      cornerRadii=<0 0 0 0>
     }
     tintColor {colorUseRule=UseColor_Specified, specifiedColor=<rgba 1 1 1 1>}
     uVRegion {bIsValid=false, max=<0 0>, min=<0 0>}
    }
   }
   slot {class=OverlaySlot, properties {horizontalAlignment=HAlign_Center, verticalAlignment=VAlign_Center}}
  }
  VerticalBox_2 {
   widgetClassPath=&VerticalBox
   parent=Overlay_Bindings
   widget=VerticalBox_2
   slot {class=OverlaySlot, properties {horizontalAlignment=HAlign_Fill, verticalAlignment=VAlign_Fill, padding=<ltrb 50 50 50 135>}}
  }
  ListView_Subscriptions {
   widgetClassPath=/Script/FortniteUI.FortCreativeDeviceBrowserListView
   bInherited=true
   bIsVariable=true
   parent=VerticalBox_2
   widget=ListView_Subscriptions
   properties {
    entryWidgetClass=&/Cr_DvBrowser_SubscriptionEntry.Cr_DvBrowser_SubscriptionEntry_C
    horizontalEntrySpacing=8
    navigation=&self:WidgetTree.ListView_Subscriptions.WidgetNavigation_0
    verticalEntrySpacing=8
   }
   slot {class=VerticalBoxSlot, properties {size {sizeRule=Fill, value=1}}}
  }
  Blade_BackgroundBlur {
   widgetClassPath=&BackgroundBlur
   bIsVariable=true
   parent=Overlay_805
   widget=Blade_BackgroundBlur
   properties {blurStrength=3, visibility=Collapsed}
   slot {class=OverlaySlot, properties {horizontalAlignment=HAlign_Fill, verticalAlignment=VAlign_Fill}}
  }
  Button_HitBlockerRight {
   widgetClassPath=&FortTextButton
   bIsVariable=true
   parent=Overlay_805
   widget=Button_HitBlockerRight
   properties {
    bHideInputAction=true
    defaultText=""
    onButtonBaseClicked=(&self:WidgetTree.Button_HitBlockerRight.HandleOnButtonClicked)
    onButtonBaseDoubleClicked=(&self:WidgetTree.Button_HitBlockerRight.HandleOnButtonDoubleClicked)
    onButtonBaseHovered=(&self:WidgetTree.Button_HitBlockerRight.HandleOnButtonHovered)
    onButtonBaseUnhovered=(&self:WidgetTree.Button_HitBlockerRight.HandleOnButtonUnhovered)
    onSelectedChangedBase=(&self:WidgetTree.Button_HitBlockerRight.HandleOnSelectedChanged)
    style=&ButtonStyle-FullyInvisible_C
   }
   slot {class=OverlaySlot, properties {horizontalAlignment=HAlign_Fill, verticalAlignment=VAlign_Fill}}
  }
  SizeBox_Options {
   widgetClassPath=&SizeBox
   parent=Overlay_805
   widget=SizeBox_Options
   properties {bOverride_WidthOverride=true, widthOverride=450}
   slot {class=OverlaySlot, properties {verticalAlignment=VAlign_Fill}}
  }
  Overlay_56 {widgetClassPath=&Overlay, parent=SizeBox_Options, widget=Overlay_56, slot {class=SizeBoxSlot}}
  Blade_SortAndFilter {
   widgetClassPath=/Game/Athena/UI/Customization/AthenaCustomizationPickerSortAndFilterBlade
   bInherited=true
   bIsVariable=true
   parent=Overlay_56
   widget=Blade_SortAndFilter
   slot {class=OverlaySlot, properties {horizontalAlignment=HAlign_Fill, verticalAlignment=VAlign_Fill}}
  }
  SafeZone_Traversal {
   widgetClassPath=&SafeZone
   bIsVariable=true
   parent=Overlay_EventBrowser
   widget=SafeZone_Traversal
   slot {class=OverlaySlot, properties {horizontalAlignment=HAlign_Fill, verticalAlignment=VAlign_Fill}}
  }
  TraversalDialog {
   widgetClassPath=&Overlay
   bInherited=true
   bIsVariable=true
   parent=SafeZone_Traversal
   widget=TraversalDialog
   properties {visibility=Collapsed}
   slot {class=SafeZoneSlot}
  }
  Traversal_BackgroundBlur {
   widgetClassPath=&BackgroundBlur
   bIsVariable=true
   parent=TraversalDialog
   widget=Traversal_BackgroundBlur
   properties {blurStrength=3, visibility=Visible}
   slot {class=OverlaySlot, properties {horizontalAlignment=HAlign_Fill, verticalAlignment=VAlign_Fill}}
  }
  Image_Backing_6 {
   widgetClassPath=&Image
   parent=Traversal_BackgroundBlur
   widget=Image_Backing_6
   properties {
    visibility=SelfHitTestInvisible
    brush {
     bIsDynamicallyLoaded=false
     drawAs=Image
     imageType=NoImage
     mirroring=NoMirror
     resourceName=None
     resourceObject=null
     tiling=NoTile
     imageSize=<32 32>
     margin=<ltrb 0 0 0 0>
     outlineSettings {
      bUseBrushTransparency=false
      roundingType=HalfHeightRadius
      width=0
      color {colorUseRule=UseColor_Specified, specifiedColor=<rgba 0 0 0 0>}
      cornerRadii=<0 0 0 0>
     }
     tintColor {colorUseRule=UseColor_Specified, specifiedColor=<rgba 0 0 0 1>}
     uVRegion {bIsValid=false, max=<0 0>, min=<0 0>}
    }
    colorAndOpacity=<rgba 1 1 1 0.75>
   }
   slot {class=BackgroundBlurSlot, properties {padding=<ltrb 0 0 0 0>}}
  }
  VerticalBox_154 {
   widgetClassPath=&VerticalBox
   parent=TraversalDialog
   widget=VerticalBox_154
   slot {class=OverlaySlot, properties {horizontalAlignment=HAlign_Fill, verticalAlignment=VAlign_Center}}
  }
  SizeBox_246 {
   widgetClassPath=&SizeBox
   parent=VerticalBox_154
   widget=SizeBox_246
   properties {bOverride_HeightOverride=true, heightOverride=175}
   slot {class=VerticalBoxSlot, properties {size {sizeRule=Automatic, value=0.25}}}
  }
  TraversalHB {
   widgetClassPath=&HorizontalBox
   parent=VerticalBox_154
   widget=TraversalHB
   slot {class=VerticalBoxSlot, properties {size {sizeRule=Fill, value=1}}}
  }
  Overlay_4 {widgetClassPath=&Overlay, parent=TraversalHB, widget=Overlay_4, slot {class=HorizontalBoxSlot}}
  SafeZone_SelectedDevice_Traversal {
   widgetClassPath=&SafeZone
   parent=Overlay_4
   widget=SafeZone_SelectedDevice_Traversal
   properties {padBottom=false, padRight=false}
   slot {class=OverlaySlot, properties {horizontalAlignment=HAlign_Fill, verticalAlignment=VAlign_Fill}}
  }
  Overlay_5 {widgetClassPath=&Overlay, parent=SafeZone_SelectedDevice_Traversal, widget=Overlay_5, slot {class=SafeZoneSlot}}
  Image_Backing_3 {
   widgetClassPath=&Image
   parent=Overlay_5
   widget=Image_Backing_3
   properties {
    visibility=SelfHitTestInvisible
    brush {
     bIsDynamicallyLoaded=false
     drawAs=Image
     imageType=NoImage
     mirroring=NoMirror
     resourceName=None
     resourceObject=null
     tiling=NoTile
     imageSize=<32 32>
     margin=<ltrb 0 0 0 0>
     outlineSettings {
      bUseBrushTransparency=false
      roundingType=HalfHeightRadius
      width=0
      color {colorUseRule=UseColor_Specified, specifiedColor=<rgba 0 0 0 0>}
      cornerRadii=<0 0 0 0>
     }
     tintColor {colorUseRule=UseColor_Specified, specifiedColor=<rgba 0.0024 0.016 0.117 1>}
     uVRegion {bIsValid=false, max=<0 0>, min=<0 0>}
    }
   }
   slot {class=OverlaySlot, properties {horizontalAlignment=HAlign_Fill, verticalAlignment=VAlign_Fill}}
  }
  SafezonePaddingReductionCB_Traversal {
   widgetClassPath=&CommonBorder
   parent=Overlay_5
   widget=SafezonePaddingReductionCB_Traversal
   properties {bReducePaddingBySafezone=true, minimumPadding=<ltrb 0 -64 0 0>, padding=<ltrb 0 0 0 0>}
   slot {class=OverlaySlot, properties {horizontalAlignment=HAlign_Fill, verticalAlignment=VAlign_Fill, padding=<ltrb 0 64 0 80>}}
  }
  Overlay_6 {
   widgetClassPath=&Overlay
   parent=SafezonePaddingReductionCB_Traversal
   widget=Overlay_6
   slot {class=BorderSlot, properties {padding=<ltrb 0 0 0 0>}}
  }
  SizeBox_SelectedDevice {
   widgetClassPath=&SizeBox
   parent=Overlay_6
   widget=SizeBox_SelectedDevice
   properties {bOverride_WidthOverride=true, widthOverride=775}
   slot {class=OverlaySlot, properties {horizontalAlignment=HAlign_Fill, verticalAlignment=VAlign_Fill}}
  }
  Overlay_7 {widgetClassPath=&Overlay, parent=SizeBox_SelectedDevice, widget=Overlay_7, slot {class=SizeBoxSlot}}
  HorizontalBox_1 {
   widgetClassPath=&HorizontalBox
   parent=Overlay_7
   widget=HorizontalBox_1
   slot {class=OverlaySlot, properties {horizontalAlignment=HAlign_Fill, verticalAlignment=VAlign_Fill}}
  }
  SizeBox_LeftPanel_Traversal {
   widgetClassPath=&SizeBox
   parent=HorizontalBox_1
   widget=SizeBox_LeftPanel_Traversal
   properties {bOverride_WidthOverride=true, widthOverride=775}
   slot {class=HorizontalBoxSlot, properties {horizontalAlignment=HAlign_Left}}
  }
  VerticalBox {
   widgetClassPath=&VerticalBox
   parent=SizeBox_LeftPanel_Traversal
   widget=VerticalBox
   slot {class=SizeBoxSlot, properties {horizontalAlignment=HAlign_Left}}
  }
  SizeBox_Title_1 {
   widgetClassPath=&SizeBox
   parent=VerticalBox
   widget=SizeBox_Title_1
   properties {bOverride_HeightOverride=true, heightOverride=100}
   slot {class=VerticalBoxSlot}
  }
  Overlay_8 {widgetClassPath=&Overlay, parent=SizeBox_Title_1, widget=Overlay_8, slot {class=SizeBoxSlot}}
  Border_1 {
   widgetClassPath=&Border
   parent=Overlay_8
   widget=Border_1
   properties {brushColor=<rgba 0.0024 0.016 0.117 1>, padding=<ltrb 0 0 0 0>}
   slot {class=OverlaySlot, properties {horizontalAlignment=HAlign_Fill, verticalAlignment=VAlign_Fill}}
  }
  CommonTextBlock {
   widgetClassPath=&CommonTextBlock
   parent=Overlay_8
   widget=CommonTextBlock
   properties {
    justification=Center
    text=DEVICE
    font {
     bForceMonospaced=false
     bMaterialIsStencil=false
     fontMaterial=null
     fontObject=&Burbank
     letterSpacing=0
     monospacedWidth=1
     size=40
     skewAmount=0
     typefaceFontName=BurbankRegular-Black
     outlineSettings {
      bApplyOutlineToDropShadows=false
      bMiteredCorners=false
      bSeparateFillAlpha=false
      outlineMaterial=&M-FontShadow
      outlineSize=0
      outlineColor=<rgba 0 0 0 1>
     }
    }
    shadowOffset=<4.9266 5.2555>
   }
   slot {class=OverlaySlot, properties {verticalAlignment=VAlign_Bottom, padding=<ltrb 50 0 0 0>}}
  }
  SizeBox_DeviceList_1 {
   widgetClassPath=&SizeBox
   parent=VerticalBox
   widget=SizeBox_DeviceList_1
   properties {bOverride_HeightOverride=true, heightOverride=690}
   slot {class=VerticalBoxSlot, properties {size {sizeRule=Fill, value=1}}}
  }
  HorizontalBox_Devices_1 {
   widgetClassPath=&HorizontalBox
   parent=SizeBox_DeviceList_1
   widget=HorizontalBox_Devices_1
   properties {
    clipping=ClipToBounds
    navigation=&self:WidgetTree.HorizontalBox_Devices_1.WidgetNavigation_0
   }
   slot {class=SizeBoxSlot}
  }
  SizeBox_1 {
   widgetClassPath=&SizeBox
   parent=HorizontalBox_Devices_1
   widget=SizeBox_1
   properties {bOverride_WidthOverride=true, heightOverride=620, widthOverride=760}
   slot {class=HorizontalBoxSlot, properties {horizontalAlignment=HAlign_Left}}
  }
  Overlay_DevicesList_1 {widgetClassPath=&Overlay, parent=SizeBox_1, widget=Overlay_DevicesList_1, slot {class=SizeBoxSlot}}
  Image_Backing_4 {
   widgetClassPath=&Image
   parent=Overlay_DevicesList_1
   widget=Image_Backing_4
   properties {
    visibility=SelfHitTestInvisible
    brush {
     bIsDynamicallyLoaded=false
     drawAs=Image
     imageType=NoImage
     mirroring=NoMirror
     resourceName=None
     resourceObject=null
     tiling=NoTile
     imageSize=<32 32>
     margin=<ltrb 0 0 0 0>
     outlineSettings {
      bUseBrushTransparency=false
      roundingType=HalfHeightRadius
      width=0
      color {colorUseRule=UseColor_Specified, specifiedColor=<rgba 0 0 0 0>}
      cornerRadii=<0 0 0 0>
     }
     tintColor {colorUseRule=UseColor_Specified, specifiedColor=<rgba 0.0024 0.016 0.117 1>}
     uVRegion {bIsValid=false, max=<0 0>, min=<0 0>}
    }
   }
   slot {class=OverlaySlot, properties {horizontalAlignment=HAlign_Fill, verticalAlignment=VAlign_Fill}}
  }
  VerticalBox_ListView_1 {
   widgetClassPath=&VerticalBox
   parent=Overlay_DevicesList_1
   widget=VerticalBox_ListView_1
   slot {class=OverlaySlot, properties {horizontalAlignment=HAlign_Fill, verticalAlignment=VAlign_Fill, padding=<ltrb 50 0 12 0>}}
  }
  ListViewHeader_Legend_Traversal {
   widgetClassPath=&SizeBox
   bIsVariable=true
   parent=VerticalBox_ListView_1
   widget=ListViewHeader_Legend_Traversal
   properties {bOverride_HeightOverride=true, heightOverride=60, widthOverride=800}
   slot {class=VerticalBoxSlot}
  }
  HorizontalBox_HeaderDevices_Traversal {
   widgetClassPath=&HorizontalBox
   parent=ListViewHeader_Legend_Traversal
   widget=HorizontalBox_HeaderDevices_Traversal
   properties {visibility=Visible}
   slot {class=SizeBoxSlot, properties {horizontalAlignment=HAlign_Right}}
  }
  Image_Events_1 {
   widgetClassPath=&Image
   parent=HorizontalBox_HeaderDevices_Traversal
   widget=Image_Events_1
   properties {brush {
     bIsDynamicallyLoaded=false
     drawAs=Image
     imageType=NoImage
     mirroring=NoMirror
     resourceName=None
     resourceObject=&MI_UI_Binding_EventNormal
     tiling=NoTile
     imageSize=<42 42>
     margin=<ltrb 0 0 0 0>
     outlineSettings {
      bUseBrushTransparency=false
      roundingType=HalfHeightRadius
      width=0
      color {colorUseRule=UseColor_Specified, specifiedColor=<rgba 0 0 0 0>}
      cornerRadii=<0 0 0 0>
     }
     tintColor {colorUseRule=UseColor_Specified, specifiedColor=<rgba 1 1 1 1>}
     uVRegion {bIsValid=false, max=<0 0>, min=<0 0>}
    }}
   slot {class=HorizontalBoxSlot, properties {horizontalAlignment=HAlign_Center, verticalAlignment=VAlign_Center, padding=<ltrb 0 0 32 0>}}
  }
  Spacer {
   widgetClassPath=&Spacer
   parent=HorizontalBox_HeaderDevices_Traversal
   widget=Spacer
   properties {size=<30 1>}
   slot {class=HorizontalBoxSlot}
  }
  Image_Functions_1 {
   widgetClassPath=&Image
   parent=HorizontalBox_HeaderDevices_Traversal
   widget=Image_Functions_1
   properties {brush {
     bIsDynamicallyLoaded=false
     drawAs=Image
     imageType=NoImage
     mirroring=NoMirror
     resourceName=None
     resourceObject=&MI_UI_Binding_FunctionNormal
     tiling=NoTile
     imageSize=<42 42>
     margin=<ltrb 0 0 0 0>
     outlineSettings {
      bUseBrushTransparency=false
      roundingType=HalfHeightRadius
      width=0
      color {colorUseRule=UseColor_Specified, specifiedColor=<rgba 0 0 0 0>}
      cornerRadii=<0 0 0 0>
     }
     tintColor {colorUseRule=UseColor_Specified, specifiedColor=<rgba 1 1 1 1>}
     uVRegion {bIsValid=false, max=<0 0>, min=<0 0>}
    }}
   slot {class=HorizontalBoxSlot, properties {horizontalAlignment=HAlign_Center, verticalAlignment=VAlign_Center, padding=<ltrb 0 0 18 0>}}
  }
  ListView_Traversal_Device {
   widgetClassPath=&CommonListView
   bInherited=true
   bIsVariable=true
   parent=VerticalBox_ListView_1
   widget=ListView_Traversal_Device
   properties {
    entryWidgetClass=&Cr_DvBrowserDvEntry_C
    horizontalEntrySpacing=8
    navigation=&self:WidgetTree.ListView_Traversal_Device.WidgetNavigation_0
    numDesignerPreviewEntries=1
    verticalEntrySpacing=8
   }
   slot {class=VerticalBoxSlot, properties {padding=<ltrb 0 0 0 20>, size {sizeRule=Fill, value=1}}}
  }
  SafeZone_Right_Traversal {
   widgetClassPath=&SafeZone
   parent=TraversalHB
   widget=SafeZone_Right_Traversal
   properties {padBottom=false, padLeft=false}
   slot {class=HorizontalBoxSlot, properties {size {sizeRule=Fill, value=1}}}
  }
  Overlay_10 {widgetClassPath=&Overlay, parent=SafeZone_Right_Traversal, widget=Overlay_10, slot {class=SafeZoneSlot}}
  SizeBox_RightPanel_1 {
   widgetClassPath=&SizeBox
   parent=Overlay_10
   widget=SizeBox_RightPanel_1
   properties {bOverride_WidthOverride=true, widthOverride=1162}
   slot {class=OverlaySlot, properties {horizontalAlignment=HAlign_Fill, verticalAlignment=VAlign_Fill}}
  }
  Overlay_Bindings_1 {widgetClassPath=&Overlay, parent=SizeBox_RightPanel_1, widget=Overlay_Bindings_1, slot {class=SizeBoxSlot}}
  Image_Backing_5 {
   widgetClassPath=&Image
   parent=Overlay_Bindings_1
   widget=Image_Backing_5
   properties {
    visibility=SelfHitTestInvisible
    brush {
     bIsDynamicallyLoaded=false
     drawAs=Image
     imageType=NoImage
     mirroring=NoMirror
     resourceName=None
     resourceObject=null
     tiling=NoTile
     imageSize=<32 32>
     margin=<ltrb 0 0 0 0>
     outlineSettings {
      bUseBrushTransparency=false
      roundingType=HalfHeightRadius
      width=0
      color {colorUseRule=UseColor_Specified, specifiedColor=<rgba 0 0 0 0>}
      cornerRadii=<0 0 0 0>
     }
     tintColor {colorUseRule=UseColor_Specified, specifiedColor=<rgba 0.0009 0.006 0.0395 1>}
     uVRegion {bIsValid=false, max=<0 0>, min=<0 0>}
    }
    colorAndOpacity=<rgba 1 1 1 0.8>
   }
   slot {class=OverlaySlot, properties {horizontalAlignment=HAlign_Fill, verticalAlignment=VAlign_Fill}}
  }
  Image_Spinner_Traversal_Subscriptions {
   widgetClassPath=&Image
   bIsVariable=true
   parent=Overlay_Bindings_1
   widget=Image_Spinner_Traversal_Subscriptions
   properties {
    visibility=Collapsed
    brush {
     bIsDynamicallyLoaded=false
     drawAs=Image
     imageType=NoImage
     mirroring=NoMirror
     resourceName=None
     resourceObject=&MI_UIKit_Monochrome_Throbber
     tiling=NoTile
     imageSize=<64 64>
     margin=<ltrb 0 0 0 0>
     outlineSettings {
      bUseBrushTransparency=false
      roundingType=HalfHeightRadius
      width=0
      color {colorUseRule=UseColor_Specified, specifiedColor=<rgba 0 0 0 0>}
      cornerRadii=<0 0 0 0>
     }
     tintColor {colorUseRule=UseColor_Specified, specifiedColor=<rgba 1 1 1 1>}
     uVRegion {bIsValid=false, max=<0 0>, min=<0 0>}
    }
   }
   slot {class=OverlaySlot, properties {horizontalAlignment=HAlign_Center, verticalAlignment=VAlign_Center}}
  }
  VerticalBox_1 {
   widgetClassPath=&VerticalBox
   parent=Overlay_Bindings_1
   widget=VerticalBox_1
   slot {class=OverlaySlot, properties {horizontalAlignment=HAlign_Fill, verticalAlignment=VAlign_Fill, padding=<ltrb 50 50 50 50>}}
  }
  SizeBox_210 {
   widgetClassPath=&SizeBox
   parent=VerticalBox_1
   widget=SizeBox_210
   properties {bOverride_MinDesiredHeight=true, minDesiredHeight=53}
   slot {class=VerticalBoxSlot}
  }
  ListViewFocusedHeader_Message {
   widgetClassPath=&SizeBox
   bIsVariable=true
   parent=VerticalBox_1
   widget=ListViewFocusedHeader_Message
   properties {bOverride_HeightOverride=true, heightOverride=60, visibility=Collapsed, widthOverride=800}
   slot {class=VerticalBoxSlot}
  }
  CommonTextBlock_Status_Traversal {
   widgetClassPath=&CommonTextBlock
   bIsVariable=true
   parent=ListViewFocusedHeader_Message
   widget=CommonTextBlock_Status_Traversal
   properties {
    bAutoCollapseWithEmptyText=true
    justification=Center
    style=&TextStyle_BurbankSmall_S_LightBLue_C
    text=No bound devices found
    colorAndOpacity {colorUseRule=UseColor_Specified, specifiedColor=<rgba 0.2051 0.6172 0.9473 1>}
    font {
     bForceMonospaced=false
     bMaterialIsStencil=false
     fontMaterial=null
     fontObject=&BurbankSmall
     letterSpacing=0
     monospacedWidth=1
     size=18
     skewAmount=0
     typefaceFontName=Black
     outlineSettings {
      bApplyOutlineToDropShadows=false
      bMiteredCorners=false
      bSeparateFillAlpha=false
      outlineMaterial=null
      outlineSize=0
      outlineColor=<rgba 0 0 0 1>
     }
    }
    shadowOffset=<0 0>
   }
   slot {class=SizeBoxSlot, properties {verticalAlignment=VAlign_Bottom}}
  }
  ListView_Traversal_Subscriptions {
   widgetClassPath=&FortCreativeDeviceBrowserListView
   bInherited=true
   bIsVariable=true
   parent=VerticalBox_1
   widget=ListView_Traversal_Subscriptions
   properties {
    entryWidgetClass=&Cr_DvBrowser_SubscriptionEntry_C
    horizontalEntrySpacing=8
    navigation=&self:WidgetTree.ListView_Traversal_Subscriptions.WidgetNavigation_0
    numDesignerPreviewEntries=6
    verticalEntrySpacing=8
   }
   slot {class=VerticalBoxSlot, properties {verticalAlignment=VAlign_Center, size {sizeRule=Fill, value=1}}}
  }
  SizeBox_2 {
   widgetClassPath=&SizeBox
   parent=VerticalBox_1
   widget=SizeBox_2
   properties {bOverride_MinDesiredHeight=true, minDesiredHeight=53}
   slot {class=VerticalBoxSlot}
  }
  SizeBox {
   widgetClassPath=&SizeBox
   parent=VerticalBox_154
   widget=SizeBox
   properties {bOverride_HeightOverride=true, heightOverride=175}
   slot {class=VerticalBoxSlot, properties {size {sizeRule=Automatic, value=0.25}}}
  }
  SafeZone_ButtonBar {
   widgetClassPath=&SafeZone
   parent=Overlay_EventBrowser
   widget=SafeZone_ButtonBar
   slot {class=OverlaySlot, properties {horizontalAlignment=HAlign_Right, verticalAlignment=VAlign_Bottom}}
  }
  Overlay_1 {
   widgetClassPath=&Overlay
   parent=SafeZone_ButtonBar
   widget=Overlay_1
   slot {class=SafeZoneSlot, properties {padding=<ltrb 0 0 0 35>}}
  }
  HorizontalBox_0 {
   widgetClassPath=&HorizontalBox
   parent=Overlay_1
   widget=HorizontalBox_0
   properties {navigation=&self:WidgetTree.HorizontalBox_0.WidgetNavigation_0}
   slot {class=OverlaySlot, properties {horizontalAlignment=HAlign_Right, verticalAlignment=VAlign_Fill, padding=<ltrb 84 24 84 0>}}
  }
  Button_Select {
   widgetClassPath=/FortUILibrary/UIKit/Buttons/WBP_UIKit_Button_Quiet
   bInherited=true
   bIsVariable=true
   parent=HorizontalBox_0
   widget=Button_Select
   properties {
    minHeight=48
    text=Details
    visibility=HitTestInvisible
    triggeringInputAction {dataTable=&InputActionDataTable, rowName=GenericForward}
   }
   slot {class=HorizontalBoxSlot, properties {horizontalAlignment=HAlign_Center, verticalAlignment=VAlign_Center}}
  }
  Button_Teleport {
   widgetClassPath=&WBP_UIKit_Button_Quiet
   bInherited=true
   bIsVariable=true
   parent=HorizontalBox_0
   widget=Button_Teleport
   properties {
    bRequiresHold=true
    minHeight=48
    text=Hold to Teleport
    visibility=Collapsed
    triggeringInputAction {dataTable=/FortUILibrary/Input/CreativeInputActionDataTable, rowName=Teleport}
   }
   slot {class=HorizontalBoxSlot, properties {horizontalAlignment=HAlign_Center, verticalAlignment=VAlign_Center, padding=<ltrb 32 0 0 0>}}
  }
  Button_Back {
   widgetClassPath=&WBP_UIKit_Button_Regular
   bInherited=true
   bIsVariable=true
   parent=HorizontalBox_0
   widget=Button_Back
   properties {minHeight=54, minWidth=264, text=Close, triggeringInputAction {dataTable=&InputActionDataTable, rowName=GenericBack}}
   slot {class=HorizontalBoxSlot, properties {horizontalAlignment=HAlign_Center, verticalAlignment=VAlign_Center, padding=<ltrb 50 0 0 0>}}
  }
  Button_CloseSortAndFilterBlade {
   widgetClassPath=&WBP_UIKit_Button_Regular
   bInherited=true
   bIsVariable=true
   parent=HorizontalBox_0
   widget=Button_CloseSortAndFilterBlade
   properties {minHeight=54, minWidth=264, text=Apply, triggeringInputAction {dataTable=&InputActionDataTable, rowName=GenericBack}}
   slot {class=HorizontalBoxSlot, properties {horizontalAlignment=HAlign_Center, verticalAlignment=VAlign_Center, padding=<ltrb 50 0 0 0>}}
  }
 ]
}
